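# Translucent, unlit, two-sided edge-glow material for Niagara mesh particles in the DerangedMile pickaxe effects.
# Particle color RGB passes through MF_FX_EmissiveAdjustments for glow, while particle alpha drives opacity.
# A procedural UV Shape mask combines a soft sphere-mask gradient along U with optional start and end fades along V.
# Opacity multiplies that mask by a smoothstepped camera-to-normal facing fade, depth fade and a scalar multiplier.
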Material /BRCosmetics/Weapons/FORT_Melee/Pickaxe_DerangedMile/FX/M_FX_EdgeGlow
graph {
 properties {
  bAutomaticallySetUsageInEditor=false
  bUsedWithNiagaraMeshParticles=true
  blendMode=BLEND_Translucent
  shadingModel=MSM_Unlit
  twoSided=true
  usedShadingModels=Unlit
 }
 nodes [
  Constant() {r=1}
  n1 = ParticleColor()
  MaterialFunctionCall(Emissive=@n1.RGB) {materialFunction=&MF_FX_EmissiveAdjustments}
  n2 = DotProduct(A=CameraVectorWS(), B=VertexNormalWS())
  n3 = Multiply(
   A=@n1.A
   B=SmoothStep(
    Min=ScalarParameter() {parameterName=View Angle Fade Min, sliderMax=1}
    Max=ScalarParameter() {defaultValue=1, parameterName=View Angle Fade Max, sliderMax=1}
    Value=Abs(None=@n2)
   )
  )
  n4 = NamedRerouteUsage()
  n5 = ScalarParameter() {defaultValue=5, parameterName=DepthFadeDistance}
  Multiply(A=Multiply(A=Multiply(A=@n3, B=@n4), B=DepthFade(FadeDistance=@n5)), B=ScalarParameter() {defaultValue=1, parameterName=OpacityMult})
  n6 = TextureCoordinate()
  n7 = ComponentMask(None=@n6) {r=true}
  n8 = ScalarParameter() {defaultValue=1, parameterName=GradientPosition}
  n9 = ScalarParameter() {defaultValue=1, parameterName=RadiusX}
  n10 = Power(Base=Multiply(A=SphereMask(A=@n7, B=@n8, Radius=@n9) {hardnessPercent=0}), Exp=ScalarParameter() {defaultValue=2, parameterName=PowerX})
  n11 = ComponentMask(None=@n6) {g=true}
  n12 = Power(Base=OneMinus(None=@n11), Exp=ScalarParameter() {defaultValue=1, group=StartFade, parameterName=StartFadePower})
  n13 = ScalarParameter() {defaultValue=5, group=StartFade, parameterName=StartFadeMult}
  n14 = LinearInterpolate(B=Saturate(None=Multiply(A=@n12, B=@n13)), Alpha=ScalarParameter() {group=StartFade, parameterName=UseStartFade?}) {constA=1}
  n15 = ScalarParameter() {defaultValue=1, group=EndFade, parameterName=EndFadePower}
  n16 = Saturate(None=Multiply(A=Power(Base=@n11, Exp=@n15), B=ScalarParameter() {defaultValue=5, group=EndFade, parameterName=EndFadeMult}))
  n17 = ScalarParameter() {group=EndFade, parameterName=UseEndFade?}
  NamedRerouteDeclaration(None=Multiply(A=Multiply(A=@n10, B=@n14), B=LinearInterpolate(B=@n16, Alpha=@n17) {constA=1})) {name=Shape, nodeColor=<rgba 0 1 0.5078 1>}
 ]
}
