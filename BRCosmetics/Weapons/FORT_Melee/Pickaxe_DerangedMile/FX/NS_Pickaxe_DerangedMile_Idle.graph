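# Niagara idle effect for the DerangedMile pickaxe, layering edge glow meshes, four base and side glow sprites, rising heat particles and drifting mesh pulses.
# System.CurrentIndex acts as a stage selector for glow color, alpha and visibility, and gates pulse and heat spawning, with heat also suppressed while swinging.
# User.Variant picks the palette, including warm and blue-violet sets, with System.ColorLerp blending shades within them.
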NiagaraSystem /BRCosmetics/Weapons/FORT_Melee/Pickaxe_DerangedMile/FX/NS_Pickaxe_DerangedMile_Idle
dependencies {
 usedDataInterfaces [
  &NiagaraDataInterfaceCurve
  &NiagaraDataInterfaceSpriteRendererInfo
  &NiagaraDataInterfaceColorCurve
 ]
 usedDynamicInputs [
  &ColorFromCurve
  &FNTimeOfDayScalar
  &Add_Float
  &Sine
  &SetBoolByFloatComparison
 ]
 usedModules [
  &SystemState
  &self:SystemSpawnScript.SystemScriptSource.SystemScriptGraph.NiagaraNodeAssignment_1.SetVariables_#1
  &self:SystemSpawnScript.SystemScriptSource.SystemScriptGraph.NiagaraNodeAssignment_0.SetVariables_#2
  &self:SystemSpawnScript.SystemScriptSource.SystemScriptGraph.NiagaraNodeAssignment_2.SetVariables_#3
 ]
}
emitters {
 GlowEdge {
  emitterUpdateScript [
   &EmitterState()
   &SpawnBurst_Instantaneous(Use Spawn Probability=false, UseLoopCountLimit=true)
  ]
  particleSpawnScript [&InitializeParticle(Lifetime=5, Mesh Scale Mode=::NewEnumerator4, UsePositionOffset=false, Write Material Random=true)]
  particleUpdateScript [
   &ParticleState(Kill Particles When Lifetime Has Elapsed=false, Let Infinitely Lived Particles Die When Emitter Deactivates=true)
   &ScaleColor(
    Scale Alpha=&Multiply_Float(
     A=&MergeHelper_FloatSwitch(0=1, 1=0.15, 2=0.15, 3=0.15, Index=$System.CurrentIndex, Num=::NewEnumerator2)
     B=&FN_FrontEndFloatModifier(Gameplay=1, Front End=1)
    )
    Scale RGB=&MakeVectorFromLinearColorRGB(Color=&MergeHelper_ColorSwitch(
      0=&MergeHelper_ColorSwitch(
       0=<rgba 1 1 1 1>
       1=<rgba 0.956 0 0 1>
       2=<rgba 1 0.159 0.0123 1>
       3=&LerpLinearColors(StartColor=<rgba 1 0.1946 0.0168 1>, EndColor=<rgba 1.5 0.2919 0.0252 1>, LerpFactor=$System.ColorLerp)
       Index=$System.CurrentIndex
       Num=::NewEnumerator2
      )
      1=&MergeHelper_ColorSwitch(
       0=<rgba 1 1 1 1>
       1=<rgba 0.0976 0.0284 0.807 1>
       2=<rgba 0.0976 0.0284 0.807 1>
       3=&LerpLinearColors(StartColor=<rgba 0.0976 0.0284 0.807 1>, EndColor=<rgba 0.1814 0.0528 1.5 1>, LerpFactor=$System.ColorLerp)
       Index=$System.CurrentIndex
       Num=::NewEnumerator2
      )
      2=?
      3=?
      Index=&MakeIntFromFloat(Float=$User.Variant)
      Num=::NewEnumerator0
     ))
    ScaleA=true
    ScaleRGB=true
   )
   /SetVariables(Particles.VisibilityTag=&MakeCustomIntFromBool(
     Bool=&SetBoolByIntComparison(A=$System.CurrentIndex, B=0, Comparison Type=::NewEnumerator4)
     True INT=1
     False INT=0
    )) {isSetParameters=true}
  ]
  rendererClasses [&NiagaraMeshRendererProperties]
  renderers [
   {rendererClass=&NiagaraMeshRendererProperties, rendererIndex=0}
   {rendererClass=&NiagaraMeshRendererProperties, rendererIndex=1}
   {rendererClass=&NiagaraMeshRendererProperties, rendererIndex=2}
  ]
 }
 Glow_BaseAndSide {
  emitterUpdateScript [
   &EmitterState(Life Cycle Mode=::NewEnumerator1, Loop Behavior=::NewEnumerator1, Loop Duration Mode=::NewEnumerator1, UseLoopDelay=false)
   &SpawnBurst_Instantaneous(Spawn Count=4, Use Spawn Probability=false, UseLoopCountLimit=true)
  ]
  particleSpawnScript [
   &PartitionParticles(Number Of Partitions=4)
   &InitializeParticle(
    Mesh Scale Mode=::NewEnumerator3
    Mesh Uniform Scale=&Multiply_Float(A=1, B=$User.FXScale)
    Position Mode=::NewEnumerator0
    Sprite Rotation Mode=::NewEnumerator0
    Sprite Size=&MergeHelper_Vector2Switch(0=<7 12>, 1=<7 12>, 2=<5 11>, 3=<5 9>, 4=?, Index=$Particles.PartitionParticles.Partition, Num=::NewEnumerator2)
    Sprite Size Mode=::NewEnumerator4
    Write Material Random=true
   )
   &SpriteFacingAndAlignment(
    Sprite Alignment=&MergeHelper_VectorSwitch(0=<-1.3 0 1>, 1=<-0.3 0 1>, 2=<0 0 1>, 3=<0 0 1>, Index=$Particles.PartitionParticles.Partition, Num=::NewEnumerator2)
    Sprite Facing=<0.7068 0.07 0.3>
    WriteAlignment=true
    WriteFacing=false
   )
   &SystemLocation(
    MaskSpawnGroup=false
    Offset=&MergeHelper_VectorSwitch(0=<-4.4 0 -11.2>, 1=<-6 0 -8.2>, 2=<0 0 20.5>, 3=<0 0 12.5>, Index=$Particles.PartitionParticles.Partition, Num=::NewEnumerator2)
   )
   &CameraOffset(Camera Offset Amount=2)
  ]
  particleUpdateScript [
   &ParticleState(Kill Particles When Lifetime Has Elapsed=false, Let Infinitely Lived Particles Die When Emitter Deactivates=true)
   &DynamicMaterialParameters(
    Index 0 Param 1=2
    Module.Index0Param0Write=true
    Module.Index0Param1Write=false
    Module.Index0Param2Write=false
    Module.Index0Param3Write=false
   )
   &ScaleColor(
    Scale Alpha=&Multiply_Float(
     A=&Multiply_Float(
      A=&MergeHelper_FloatSwitch(0=0.5, 1=0.5, 2=3, 3=3, Index=$Particles.PartitionParticles.Partition, Num=::NewEnumerator2)
      B=&MergeHelper_FloatSwitch(0=1, 1=1, 2=1, 3=0.8, Index=$System.CurrentIndex, Num=::NewEnumerator2)
     )
     B=&FN_FrontEndFloatModifier(Gameplay=1, Front End=1.1)
    )
    Scale RGB=&MakeVectorFromLinearColorRGB(Color=&MergeHelper_ColorSwitch(
      0=&MergeHelper_ColorSwitch(
       0=<rgba 1 1 1 1>
       1=<rgba 0.956 0 0 1>
       2=<rgba 1 0.159 0.0123 1>
       3=&LerpLinearColors(StartColor=<rgba 1 0.1946 0.0168 1>, EndColor=<rgba 1.5 0.2919 0.0252 1>, LerpFactor=$System.ColorLerp)
       Index=$System.CurrentIndex
       Num=::NewEnumerator2
      )
      1=&MergeHelper_ColorSwitch(
       0=<rgba 1 1 1 1>
       1=<rgba 0.0976 0.0284 0.807 1>
       2=<rgba 0.0976 0.0284 0.807 1>
       3=&LerpLinearColors(StartColor=<rgba 0.0976 0.0284 0.807 1>, EndColor=<rgba 0.1814 0.0528 1.5 1>, LerpFactor=$System.ColorLerp)
       Index=$System.CurrentIndex
       Num=::NewEnumerator2
      )
      2=?
      3=?
      Index=&MakeIntFromFloat(Float=$User.Variant)
      Num=::NewEnumerator0
     ))
    ScaleA=true
    ScaleRGB=true
   )
   /SetVariables(Particles.VisibilityTag=&MakeCustomIntFromBool(
     Bool=&SetBoolByIntComparison(A=$System.CurrentIndex, B=0, Comparison Type=::NewEnumerator4)
     True INT=1
     False INT=0
    )) {isSetParameters=true}
  ]
  rendererClasses [&NiagaraSpriteRendererProperties]
  renderers [{rendererClass=&NiagaraSpriteRendererProperties}]
 }
 HeatParticles {
  emitterUpdateScript [
   &EmitterState(Life Cycle Mode=::NewEnumerator1, Loop Duration=2, UseLoopDelay=false)
   &SpawnRate(
    Spawn Probability=&MakeCustomFloatFromBool(
     Bool=&Boolean_OrOperation(
      Bool1=$User.bisSwinging
      Bool2=&SetBoolByIntComparison(A=$System.CurrentIndex, B=3, Comparison Type=::NewEnumerator5)
     )
     True Float=0
     False Float=1
    )
    SpawnRate=10
    Use Spawn Probability=true
   )
  ]
  particleSpawnScript [
   /SetVariables(Particles.SpawnGroup=&MakeIntFromFloat(Float=&UniformAOrBFloat(
      A=0
      B=&UniformAOrBFloat(A=1, B=2, Distribution Weight=0.2, Evaluation Type=::NewEnumerator0, Recalculate Random Each Loop=true, Override Seed=false)
      Distribution Weight=0.2
      Evaluation Type=::NewEnumerator0
      Recalculate Random Each Loop=true
      Override Seed=false
     ))) {isSetParameters=true}
   &InitializeParticle(
    Color=&MergeHelper_ColorSwitch(
     0=&RandomRangeLinearColor(Minimum=<rgba 1 0.1946 0.0168 1>, Maximum=<rgba 1 0.3333 0 1>)
     1=&LerpLinearColors(StartColor=<rgba 0.0352 0.196 1 1>, EndColor=<rgba 0.1814 0.0528 1.5 1>, LerpFactor=$System.ColorLerp)
     2=?
     3=?
     Index=&MakeIntFromFloat(Float=$User.Variant)
     Num=::NewEnumerator0
    )
    Lifetime Max=1.5
    Lifetime Min=0.8
    Lifetime Mode=::NewEnumerator1
    Mass Mode=::NewEnumerator3
    Mesh Scale Mode=::NewEnumerator3
    Mesh Uniform Scale=0.2
    Position Offset=<0 0 50>
    Sprite Rotation Mode=::NewEnumerator2
    Sprite Size Mode=::NewEnumerator1
    Uniform Sprite Size Max=38
    Uniform Sprite Size Min=25
    UsePositionOffset=false
    Write Material Random=true
   )
   &ShapeLocation(
    Cylinder Height=33
    Cylinder Radius=2
    MaskSpawnGroup=true
    Offset=<-5 0 20>
    Offset Coordinate Space=::Local
    Offset Mode=::NewEnumerator0
    Override Seed=false
    Shape Primitive=::NewEnumerator1
    Surface Only=false
    UseOwnerScale=true
   )
   &ShapeLocation(
    Cylinder Height=7
    Cylinder Radius=3
    MaskSpawnGroup=true
    Offset=<-1 0 38>
    Offset Coordinate Space=::Local
    Offset Mode=::NewEnumerator0
    Override Seed=false
    Shape Primitive=::NewEnumerator1
    Spawn Group Mask=1
    Surface Only=false
    UseOwnerScale=true
   ) {module=ShapeLocation001}
   &ShapeLocation(
    MaskSpawnGroup=true
    Non Uniform Scale=<1.5 1 1>
    Offset=<-5 0 -11>
    Offset Coordinate Space=::Local
    Offset Mode=::NewEnumerator0
    Override Seed=false
    Spawn Group Mask=2
    Sphere Radius=2
    UseOwnerScale=true
   ) {module=ShapeLocation002}
   &SubUVAnimation(
    Evaluation Type=::NewEnumerator0
    "Sprite Renderer "=DataInterface() {propertyValues {SpriteRenderer=&self:HeatParticles_2.NiagaraSpriteRendererProperties_0}}
    SubUV Animation Mode:ENiagara_SubUVLookupModeV2=::NewEnumerator2
    UseEndFrame=false
    UseStartFrame=false
   )
  ]
  particleUpdateScript [
   &ParticleState()
   &ScaleSpriteSize()
   &ScaleColor(
    Scale Alpha=&FloatFromCurve(
     FloatCurve=DataInterface() {propertyValues {Curve {keys [<at 0 0 RCIM_Cubic>,<at 0.4 1 RCIM_Cubic>,<at 1 0 RCIM_Cubic>]}}}
     Scale Curve=&RandomRangeFloat(Minimum=0.06, Maximum=0.09)
    )
    Scale RGB=&VectorFromFloat(Value=&Lerp_Float(
      A=3
      B=0.1
      Alpha=&FloatFromCurve(FloatCurve=DataInterface() {propertyValues {Curve {keys [<at 0 0>,<at 1 1>]}}})
     ))
    ScaleA=true
    ScaleRGB=true
   )
   &AccelerationForce(Acceleration=<0 0 4>)
   &SolveForcesAndVelocity(Clamp Velocity=false, Limit Acceleration=false)
   &ScaleColor(Scale Alpha=100, ScaleA=true, ScaleRGB=false) {enabled=false, module=ScaleColor001}
   &ScaleColor(
    Scale RGB=&MakeVectorFromLinearColorRGB(Color=&LerpLinearColors(StartColor=<rgba 1 0.1946 0.0168 1>, EndColor=<rgba 1 0.7011 0.107 1>, LerpFactor=$System.ColorLerp))
    ScaleA=false
    ScaleRGB=true
   ) {enabled=false, module=ScaleColor002}
  ]
  rendererClasses [&NiagaraSpriteRendererProperties, &NiagaraMeshRendererProperties]
  renderers [
   {rendererClass=&NiagaraSpriteRendererProperties}
   {rendererClass=&NiagaraMeshRendererProperties, rendererIndex=1}
  ]
 }
 Pulse {
  emitterUpdateScript [
   &EmitterState()
   &SpawnRate(
    Spawn Probability=&MakeCustomFloatFromBool(
     Bool=&SetBoolByIntComparison(A=$System.CurrentIndex, B=3, Comparison Type=::NewEnumerator4)
     True Float=1
     False Float=0
    )
    Use Spawn Probability=true
   )
  ]
  particleSpawnScript [
   &InitializeParticle(Color=<rgba 1 1 1 0.6>, Lifetime=0.6, Mesh Scale Mode=::NewEnumerator3, UsePositionOffset=false, Write Material Random=true)
   &AddVelocity(Velocity=<-5 0 3>, Velocity Speed Scale=&RandomRangeFloat(Minimum=0.8, Maximum=1))
   &DynamicMaterialParameters(
    Index 0 Param 1=&RandomRangeFloat(Minimum=0, Maximum=1)
    Module.Index0Float0Write=true
    Module.Index0Float1Write=true
    Module.Index0Float2Write=true
    Module.Index0Float3Write=true
   )
  ]
  particleUpdateScript [
   &ParticleState()
   &SolveForcesAndVelocity(Clamp Velocity=false, Limit Acceleration=false)
   &ScaleColor(
    Scale Alpha=&FloatFromCurve(
     FloatCurve=DataInterface() {propertyValues {Curve {keys [
         {interpMode=RCIM_Cubic, tangentMode=RCTM_Break, time=0.05, value=0}
         {interpMode=RCIM_Cubic, tangentMode=RCTM_Break, time=0.4, value=1}
         {interpMode=RCIM_Cubic, tangentMode=RCTM_Break, time=1, value=0}
        ]}}}
     Scale Curve=0.5
    )
    Scale RGB=&MakeVectorFromLinearColorRGB(Color=&MergeHelper_ColorSwitch(
      0=&LerpLinearColors(StartColor=<rgba 1 0.1946 0.0168 1>, EndColor=<rgba 1.5 0.2919 0.0252 1>, LerpFactor=$System.ColorLerp)
      1=&LerpLinearColors(StartColor=<rgba 0.1209 0.0352 1 1>, EndColor=<rgba 0.1814 0.0528 1.5 1>, LerpFactor=$System.ColorLerp)
      2=?
      3=?
      Index=&MakeIntFromFloat(Float=$User.Variant)
      Num=::NewEnumerator0
     ))
    ScaleA=true
    ScaleRGB=true
   )
   &ScaleMeshSize(Scale Factor=&MakeVector(
     X=1
     Y=&FloatFromCurve(FloatCurve=DataInterface() {propertyValues {Curve {keys [<at 0 1>,<at 1 1>]}}})
     Z=1
    ))
  ]
  rendererClasses [&NiagaraMeshRendererProperties]
  renderers [{rendererClass=&NiagaraMeshRendererProperties, rendererIndex=0}]
 }
}
summary {
 systemName=NS_Pickaxe_DerangedMile_Idle
 emitters [
  {emitterName=GlowEdge, rendererClasses [&NiagaraMeshRendererProperties]}
  {emitterName=Pulse, rendererClasses [&NiagaraMeshRendererProperties]}
  {emitterName=HeatParticles, rendererClasses [&NiagaraSpriteRendererProperties, &NiagaraMeshRendererProperties]}
  {emitterName=Glow_BaseAndSide, rendererClasses [&NiagaraSpriteRendererProperties]}
 ]
 userVariables [
  User.bisFootHidden {type=NiagaraBool}
  User.bIsFrontEnd {type=NiagaraBool}
  User.bIsFrontEndPreview {type=NiagaraBool}
  User.bisSwinging {type=NiagaraBool}
  User.CurrentIndex {type=NiagaraInt32}
  User.FXScale {type=NiagaraFloat, defaultValue=1, description=""}
  User.ReactivityIsOn {type=NiagaraFloat, defaultValue=1, description=""}
  User.Variant {type=NiagaraFloat, defaultValue=0, description=""}
 ]
}
