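# Niagara impact effect for the Deranged Mile pickaxe, combining a claw sprite, fire energy, directional sparks and streaks, flash elements and debris.
# Burst emitters launch particles along a system direction vector, with debris meshes and physics chosen by physical surface.
# User.Variant switches between orange and blue palettes, while System.CurrentIndex varies colors and gates some emitters.
# Front-end modifiers rescale the effect and time-of-day inputs adjust brightness.
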NiagaraSystem /BRCosmetics/Weapons/FORT_Melee/Pickaxe_DerangedMile/FX/NS_Pickaxe_DerangedMile_Impact
dependencies {
 usedDataInterfaces [
  &NiagaraDataInterfaceCurve
  &NiagaraDataInterfaceColorCurve
  &NiagaraDataInterfaceVectorCurve
  &NiagaraDataInterfaceVector2DCurve
  &NiagaraDataInterfaceSpriteRendererInfo
  &NiagaraDataInterfaceArrayFloat3
 ]
 usedDynamicInputs [&Vector2DFromFloat, /Game/Effects/Niagara/Modules/Utility/FN_DirectionVectorByPickaxeType]
 usedModules [
  &self:SystemSpawnScript.SystemScriptSource.SystemScriptGraph.NiagaraNodeAssignment_0.SetVariables_#1
  &self:SystemSpawnScript.SystemScriptSource.SystemScriptGraph.NiagaraNodeAssignment_2.SetVariables_#2
  &self:SystemSpawnScript.SystemScriptSource.SystemScriptGraph.NiagaraNodeAssignment_3.SetVariables_#3
  &self:SystemSpawnScript.SystemScriptSource.SystemScriptGraph.NiagaraNodeAssignment_5.SetVariables_#4
  &SystemState
  &self:SystemSpawnScript.SystemScriptSource.SystemScriptGraph.NiagaraNodeAssignment_4.SetVariables_#5
 ]
}
emitters {
 Claw {
  emitterUpdateScript [
   &EmitterState(Life Cycle Mode=::NewEnumerator1, Loop Behavior=::NewEnumerator1, Loop Duration=1, UseLoopDelay=false)
   &SpawnBurst_Instantaneous(
    Spawn Probability=&MakeCustomFloatFromBool(
     Bool=&SetBoolByIntComparison(A=$System.CurrentIndex, B=3, Comparison Type=::NewEnumerator5)
     True Float=0
     False Float=1
    )
    Use Spawn Probability=true
    UseLoopCountLimit=false
   )
   /SetVariables(Emitter.Color1:Vector3f=&MakeVectorFromLinearColorRGB(Color=&ColorFromCurve(
      ColorCurve=DataInterface() {propertyValues {
        AlphaCurve {keys [<at 0 1 RCIM_Cubic>]}
        BlueCurve {keys [<at 0.0667 8>,<at 0.1378 2 RCIM_Cubic>]}
        GreenCurve {keys [<at 0.0667 8>,<at 0.1378 2 RCIM_Cubic>]}
        RedCurve {keys [<at 0.0667 8>,<at 0.1378 2 RCIM_Cubic>]}
       }}
      CurveIndex=$Emitter.NormalizedLoopAge
     ))) {isSetParameters=true}
   /SetVariables(Emitter.Color2:Vector3f=&MakeVectorFromLinearColorRGB(Color=&ColorFromCurve(
      ColorCurve=DataInterface() {propertyValues {
        AlphaCurve {keys [<at 0 1 RCIM_Cubic>]}
        BlueCurve {keys [<at 0.0955 10>,<at 0.236 4>]}
        GreenCurve {keys [<at 0.0955 10>,<at 0.236 4>]}
        RedCurve {keys [<at 0.0955 10>,<at 0.236 4>]}
       }}
      CurveIndex=$Emitter.NormalizedLoopAge
     ))) {isSetParameters=true}
  ]
  particleSpawnScript [
   &InitializeParticle(Lifetime=0.25, Sprite Size=<-200 130>, Sprite Size Mode=::NewEnumerator4, UsePositionOffset=true, Write Material Random=true)
   &CameraOffset(Camera Offset Amount=&FN_FrontEndFloatModifier(Gameplay=50, Front End=0))
   &AddVelocity(
    Velocity=$System.DirectionVector
    Velocity Speed Scale=&Multiply_Float(
     A=&RandomRangeFloat(Minimum=10, Maximum=20)
     B=&FN_FrontEndFloatModifier(Gameplay=1, Front End=$User.FrontEndScale)
    )
   ) {module=AddVelocity001}
  ]
  particleUpdateScript [
   &ParticleState()
   &ScaleSpriteSize(
    Non-Uniform Curve Scale=&Multiply_Vector2D_ByFloat(
     Float=&Multiply_Float(A=0.25, B=&FN_FrontEndFloatModifier(Gameplay=1, Front End=$System.FrontEndScale))
     Vector2D=&MakeVector2D(X=1, Y=1)
    )
    Non-Uniform Curve Sprite Scale=DataInterface() {propertyValues {
      XCurve {keys [
        {arriveTangent=0.0233, interpMode=RCIM_Cubic, leaveTangent=0.0233, tangentMode=RCTM_User, time=0, value=2}
        {tangentMode=RCTM_User, time=1, value=1}
       ]}
      YCurve {keys [
        {arriveTangent=3.2417, interpMode=RCIM_Cubic, leaveTangent=3.2417, tangentMode=RCTM_User, time=0, value=1}
        {arriveTangent=2.8317, tangentMode=RCTM_User, time=1, value=5}
       ]}
     }}
    Scale Sprite Size Mode=::NewEnumerator3
   )
   &SolveForcesAndVelocity(Clamp Velocity=false, Limit Acceleration=false)
   &DynamicMaterialParameters(
    Index 0 Param 1=&FloatFromCurve(FloatCurve=DataInterface() {propertyValues {Curve {keys [
         <at 0 1>
         {arriveTangent=-1, interpMode=RCIM_Cubic, leaveTangent=-1, time=0.2, value=1}
         {arriveTangent=-1.6325, interpMode=RCIM_Cubic, leaveTangent=-1.6325, tangentMode=RCTM_User, time=1, value=0}
        ]}}})
    Module.Index0Param0Write=true
    Module.Index0Param1Write=true
    Module.Index0Param2Write=true
    Module.Index0Param3Write=true
   )
   &ScaleColor(
    Scale RGB=&MergeHelper_VectorSwitch(
     0=&MakeVectorFromLinearColorRGB(Color=&ColorFromCurve(ColorCurve=DataInterface() {propertyValues {
         AlphaCurve {keys [<at 0 1 RCIM_Cubic>,<at 1 1 RCIM_Cubic>]}
         BlueCurve {keys [<at 0.2861 0.15 RCIM_Cubic>,<at 0.7964 0.0003 RCIM_Cubic>]}
         GreenCurve {keys [<at 0.2861 0.7593 RCIM_Cubic>,<at 0.7964 0.0449 RCIM_Cubic>]}
         RedCurve {keys [<at 0.2861 1 RCIM_Cubic>,<at 0.7964 1 RCIM_Cubic>]}
        }}))
     1=&MakeVectorFromLinearColorRGB(Color=&ColorFromCurve(ColorCurve=DataInterface() {propertyValues {
         AlphaCurve {keys [<at 0 1 RCIM_Cubic>,<at 1 1 RCIM_Cubic>]}
         BlueCurve {keys [<at 0.2861 1 RCIM_Cubic>,<at 0.7964 1 RCIM_Cubic>]}
         GreenCurve {keys [<at 0.2861 0.6458 RCIM_Cubic>,<at 0.7964 0.0836 RCIM_Cubic>]}
         RedCurve {keys [<at 0.2861 0.15 RCIM_Cubic>,<at 0.7964 0.0003 RCIM_Cubic>]}
        }}))
     2=?
     3=?
     Index=&MakeIntFromFloat(Float=$User.Variant)
     Num=::NewEnumerator0
    )
    ScaleA=true
    ScaleRGB=true
   )
  ]
  rendererClasses [&NiagaraSpriteRendererProperties]
  renderers [{rendererClass=&NiagaraSpriteRendererProperties}]
 }
 DirectionalSparks {
  emitterUpdateScript [
   &EmitterState()
   &SpawnBurst_Instantaneous(Spawn Count=30, Use Spawn Probability=true, UseLoopCountLimit=false)
  ]
  particleSpawnScript [
   &InitializeParticle(
    Color=&MergeHelper_ColorSwitch(
     0=&MergeHelper_ColorSwitch(
      0=<rgba 2 0.8985 0.0123 1>
      1=&RandomRangeLinearColor(Minimum=<rgba 2 0.3436 0.0123 1>, Maximum=<rgba 2 0.8985 0.0123 1>)
      2=<rgba 2 0.8985 0.0123 1>
      3=<rgba 2 0.8985 0.0123 1>
      Index=$System.CurrentIndex
      Num=::NewEnumerator2
     )
     1=&MergeHelper_ColorSwitch(
      0=&LerpLinearColors(
       StartColor=<rgba 2 0.9414 0.2536 1>
       EndColor=<rgba 2 0.6548 0.3045 1>
       LerpFactor=&RandomRangeFloat(Minimum=0, Maximum=1)
      )
      1=&LerpLinearColors(
       StartColor=<rgba 0.2536 1.4179 2 1>
       EndColor=<rgba 0.3045 0.8696 2 1>
       LerpFactor=&RandomRangeFloat(Minimum=0, Maximum=1)
      )
      2=<rgba 0.2536 1.4179 2 1>
      3=<rgba 0.2536 1.4179 2 1>
      Index=$System.CurrentIndex
      Num=::NewEnumerator2
     )
     2=?
     3=?
     Index=&MakeIntFromFloat(Float=$User.Variant)
     Num=::NewEnumerator0
    )
    Lifetime Max=0.45
    Lifetime Min=0.05
    Lifetime Mode=::NewEnumerator1
    Sprite Size=&Multiply_Vector2D_ByFloat(
     Float=&Multiply_Float(
      A=&FN_FrontEndFloatModifier(Gameplay=1, Front End=0.4)
      B=&RandomRangeFloat(
       Minimum=&MaskFloat(Mask Source=1, Mask Parameter=$System.CurrentIndex, Mask=3, Masked Float=3, Passthrough Float=4.5)
       Maximum=5
      )
     )
     Vector2D=<1 1.5>
    )
    Sprite Size Mode=::NewEnumerator4
    UsePositionOffset=false
    Write Material Random=true
   )
   &ShapeLocation(
    Cylinder Height=80
    Cylinder Radius=2
    MaskSpawnGroup=false
    Override Seed=false
    Rotation Axis=$System.DirectionVector
    Rotation Mode=::NewEnumerator1
    Shape Primitive=::NewEnumerator1
    Surface Only=false
    UseOwnerScale=true
   )
   &CameraOffset(Camera Offset Amount=20)
   &AddVelocityFromPoint(
    Module.Use Falloff Distance=false
    Velocity Strength=&Multiply_Float(
     A=&RandomRangeFloat(Minimum=150, Maximum=200)
     B=&FN_FrontEndFloatModifier(Gameplay=1.5, Front End=$User.FrontEndScale)
    )
   )
   &AddVelocity(
    Velocity=$System.DirectionVector
    Velocity Speed Scale=&Multiply_Float(
     A=&RandomRangeFloat(Minimum=222, Maximum=700)
     B=&FN_FrontEndFloatModifier(Gameplay=1.3, Front End=$User.FrontEndScale)
    )
   )
  ]
  particleUpdateScript [
   &ParticleState()
   &AccelerationForce(Acceleration=<0 0 -500>)
   &Drag(
    Drag=&MaskFloat(
     Mask Source=1
     Mask Parameter=$System.CurrentIndex
     Mask=3
     Masked Float=4
     Passthrough Float=&RandomRangeFloat(Minimum=5, Maximum=8)
    )
    Use Linear Drag=true
    Use Rotational Drag=true
   )
   &SolveForcesAndVelocity(Clamp Velocity=false, Limit Acceleration=false)
   &ScaleSpriteSizeBySpeed(
    Max Scale Factor=/Niagara/DynamicInputs/Helpers/MaskVector2d(
     Mask Source:NiagaraInt32=1
     Mask Parameter:NiagaraInt32=$System.CurrentIndex
     Mask:NiagaraInt32=3
     Masked Vector2D:Vector2f=<0.5 10>
     Passthrough Vector2D:Vector2f=<0.2 5>
    )
    Min Scale Factor=&MaskVector2d(
     Mask Source:NiagaraInt32=1
     Mask Parameter:NiagaraInt32=$System.CurrentIndex
     Mask:NiagaraInt32=3
     Masked Vector2D:Vector2f=<0.5 1>
     Passthrough Vector2D:Vector2f=<0.2 0.2>
    )
   )
   &ScaleSpriteSize(Uniform Curve Sprite Scale=DataInterface() {propertyValues {Curve {keys [{arriveTangent=-6.725, interpMode=RCIM_Cubic, leaveTangent=-6.725, tangentMode=RCTM_User, time=0.65, value=1}, <at 1 0 RCIM_Cubic>]}}})
   &ScaleColor(
    Scale RGB=&Multiply_Vector(
     A=&VectorFromFloat(Value=&FloatFromCurve(
       FloatCurve=DataInterface() {propertyValues {Curve {keys [
           {interpMode=RCIM_Cubic, tangentMode=RCTM_User, time=0, value=1}
           <at 0.3 0.2>
           {interpMode=RCIM_Cubic, tangentMode=RCTM_User, time=1, value=0}
          ]}}}
       Scale Curve=10
      ))
     B=&MakeVectorFromLinearColorRGB(Color=&MergeHelper_ColorSwitch(
       0=<rgba 1 1 1 1>
       1=<rgba 1 1 1 1>
       2=<rgba 1 1 1 1>
       3=&ColorFromCurve(ColorCurve=DataInterface() {propertyValues {
          AlphaCurve {keys [<at 0 1 RCIM_Cubic>,<at 1 0 RCIM_Cubic>]}
          BlueCurve {keys [<at 0.4507 1 RCIM_Cubic>,<at 0.7557 0.05 RCIM_Cubic>]}
          GreenCurve {keys [<at 0.4507 1 RCIM_Cubic>,<at 0.7557 0.1608 RCIM_Cubic>]}
          RedCurve {keys [<at 0.4507 1 RCIM_Cubic>,<at 0.7557 1 RCIM_Cubic>]}
         }})
       Index=$System.CurrentIndex
       Num=::NewEnumerator2
      ))
    )
    ScaleA=true
    ScaleRGB=true
   )
   /SetVariables(
    Particles.GlowColor=&MakeLinearColorFromVectorAndFloat(
     "Vector (RGB)"=&MakeVectorFromLinearColorRGB(Color=&MergeHelper_ColorSwitch(
       0=&MergeHelper_ColorSwitch(0=<rgba 1 0.2015 0 1>, 1=<rgba 1 0.05 0 1>, 2=<rgba 1 0.2015 0 1>, 3=<rgba 1 0.19 0.1 1>, Index=$System.CurrentIndex, Num=::NewEnumerator2)
       1=&MergeHelper_ColorSwitch(0=<rgba 1 0.2015 0 1>, 1=<rgba 0 0.1667 1 1>, 2=<rgba 0 0.1667 1 1>, 3=<rgba 0 0.1667 1 1>, Index=$System.CurrentIndex, Num=::NewEnumerator2)
       2=?
       3=?
       Index=&MakeIntFromFloat(Float=$User.Variant)
       Num=::NewEnumerator0
      ))
     "Float (Alpha)"=&FloatFromCurve(FloatCurve=DataInterface() {propertyValues {Curve {keys [<at 0.5 1>,<at 1 0>]}}}, Scale Curve=0.1)
    )
    Particles.GlowSize=&Multiply_Vector2D(A=$Particles.SpriteSize, B=<8 2>)
   ) {isSetParameters=true}
  ]
  rendererClasses [&NiagaraSpriteRendererProperties]
  renderers [
   {rendererClass=&NiagaraSpriteRendererProperties}
   {rendererClass=&NiagaraSpriteRendererProperties, rendererIndex=1}
  ]
 }
 DirectionalStreaks {
  emitterUpdateScript [
   &EmitterState()
   &SpawnBurst_Instantaneous(
    Spawn Count=2
    Spawn Probability=&MakeCustomFloatFromBool(
     Bool=&SetBoolByIntComparison(A=$System.CurrentIndex, B=3, Comparison Type=::NewEnumerator4)
     True Float=0
     False Float=1
    )
    Use Spawn Probability=true
    UseLoopCountLimit=false
   )
  ]
  particleSpawnScript [
   &InitializeParticle(
    Color=&MergeHelper_ColorSwitch(
     0=&MergeHelper_ColorSwitch(
      0=&RandomRangeLinearColor(Minimum=<rgba 1 0.3476 0 1>, Maximum=<rgba 1 0.0048 0 1>)
      1=&RandomRangeLinearColor(Minimum=<rgba 1 0.3476 0 1>, Maximum=<rgba 1 0.0048 0 1>)
      2=&RandomRangeLinearColor(Minimum=<rgba 1 0.3476 0 1>, Maximum=<rgba 1 0.0048 0 1>)
      3=&RandomRangeLinearColor(Minimum=<rgba 1 0.3476 0 1>, Maximum=<rgba 1 0.0048 0 1>)
      Index=$System.CurrentIndex
      Num=::NewEnumerator2
     )
     1=&MergeHelper_ColorSwitch(
      0=&RandomRangeLinearColor(Minimum=<rgba 1 0.3476 0 1>, Maximum=<rgba 1 0.0048 0 1>)
      1=&RandomRangeLinearColor(Minimum=<rgba 0 0.3333 1 1>, Maximum=<rgba 0 0.1667 1 1>)
      2=&RandomRangeLinearColor(Minimum=<rgba 0 0.6667 1 1>, Maximum=<rgba 0 0.1667 1 1>)
      3=&RandomRangeLinearColor(Minimum=<rgba 0 0.6667 1 1>, Maximum=<rgba 0 0.1667 1 1>)
      Index=$System.CurrentIndex
      Num=::NewEnumerator2
     )
     2=?
     3=?
     Index=&MakeIntFromFloat(Float=$User.Variant)
     Num=::NewEnumerator0
    )
    Lifetime Max=0.15
    Lifetime Min=0.1
    Lifetime Mode=::NewEnumerator1
    Sprite Size=&Multiply_Vector2D(
     A=&UniformRangedVector2D(Minimum=<20 40>, Maximum=<25 70>)
     B=&Multiply_Vector2D_ByFloat(
      Float=0.8
      Vector2D=&FN_FrontEndVector2Modifier(Front End:Vector2f=$System.FrontEndDirectionalScale, Gameplay:Vector2f=<1 1>)
     )
    )
    Sprite Size Mode=::NewEnumerator4
    Sprite UV Mode=::NewEnumerator1
    UsePositionOffset=false
    Write Material Random=true
   ) {module=InitializeParticle001}
   &AddVelocity(
    Module.Use Falloff Distance=false
    Override Seed=false
    Velocity Mode=::NewEnumerator1
    Velocity Speed=&Multiply_Float(A=20, B=&FN_FrontEndFloatModifier(Gameplay=1, Front End=$User.FrontEndScale))
   ) {module=AddVelocity001}
   &SubUVAnimation(End Frame=3, SubUV Animation Mode:ENiagara_SubUVLookupMode=::NewEnumerator2)
   &AddVelocity(
    Velocity=$System.DirectionVector
    Velocity Speed Scale=&Multiply_Float(A=100, B=&FN_FrontEndFloatModifier(Gameplay=1, Front End=$User.FrontEndScale))
   )
   &CameraOffset(Camera Offset Amount=25)
  ]
  particleUpdateScript [
   &ParticleState()
   &SolveForcesAndVelocity(Clamp Velocity=false, Limit Acceleration=false)
   &ScaleSpriteSize(Scale Factor=&Vector2DFromCurve(
     Vector2Curve=DataInterface() {propertyValues {
       XCurve {keys [<at 0 1 RCIM_Cubic>,<at 1 0 RCIM_Cubic>]}
       YCurve {keys [{arriveTangent=3.9291, interpMode=RCIM_Cubic, leaveTangent=3.9291, tangentMode=RCTM_User, time=0, value=0}, <at 1 2 RCIM_Cubic>]}
      }}
     ScaleCurve=&Multiply_Vector2D_ByFloat(Float=1, Vector2D=<1.5 1>)
    ))
   &ScaleColor(
    Scale Alpha=&FloatFromCurve(
     FloatCurve=DataInterface() {propertyValues {Curve {keys [<at 0 1 RCIM_Cubic>,<at 1 0>]}}}
     Scale Curve=&FNTimeOfDayScalar(Night Scale=0.5, Morning Scale=1, Evening Scale=0.5)
    )
    Scale RGB=&VectorFromFloat(Value=&FloatFromCurve(FloatCurve=DataInterface() {propertyValues {Curve {keys [{arriveTangent=-15.8795, interpMode=RCIM_Cubic, leaveTangent=-15.8795, tangentMode=RCTM_User, time=0, value=10}, <at 1 1 RCIM_Cubic>]}}}))
    ScaleA=true
    ScaleRGB=true
   )
  ]
  rendererClasses [&NiagaraSpriteRendererProperties]
  renderers [{rendererClass=&NiagaraSpriteRendererProperties}]
 }
 Directional_FN__PhysMatDebris {
  emitterUpdateScript [
   &EmitterState()
   &SpawnBurst_Instantaneous(
    Spawn Count=&RandomRangeInt(Minimum=5, Maximum=10)
    Use Spawn Probability=false
    UseLoopCountLimit=false
   )
  ]
  particleSpawnScript [
   /SetVariables(Particles.BiasedRandom=&FloatFromCurve(
     FloatCurve=DataInterface() {propertyValues {Curve {keys [
         <at 0.2 1 RCIM_Cubic>
         {arriveTangent=-3.3394, interpMode=RCIM_Cubic, leaveTangent=-3.3394, tangentMode=RCTM_User, time=0.5, value=0.5}
         <at 0.8 0 RCIM_Cubic>
        ]}}}
     CurveIndex=&RandomRangeFloat(Minimum=0, Maximum=1)
    )) {isSetParameters=true}
   &InitializeParticle(
    Color Mode=::NewEnumerator0
    Lifetime=&RandomRangeFloat(Minimum=0.5, Maximum=1.25)
    Mass=&Lerp_Float(A=0.6, B=1.2, Alpha=$Particles.BiasedRandom)
    Mass Mode=::NewEnumerator1
    Mesh Scale=&Multiply_VectorByFloat(
     Vector=&RandomRangeVector(Minimum=<0.9 0.4 0.5>, Maximum=<1 1 1>)
     Float=&Lerp_Float(A=0.7, B=1, Alpha=$Particles.BiasedRandom)
    )
    Mesh Scale Mode=::NewEnumerator4
    UsePositionOffset=false
    Write Material Random=true
   )
   &FN_PhysicalSurface(
    Dirt Mesh Count=2
    Glass Mesh Count=3
    Ice Mesh Count=2
    Metal Gravity=&RandomRangeFloat(Minimum=-350, Maximum=-1800)
    Metal Mesh Count=2
    Metal Velocity=&RandomRangeFloat(Minimum=80, Maximum=375)
    Metal Velocity Cone Angle=130
    Physical Surface=$User.PhysicalSurface
    Stone Gravity=&RandomRangeFloat(Minimum=-700, Maximum=-1250)
    Stone Mesh Count=2
    Stone Velocity=&RandomRangeFloat(Minimum=80, Maximum=550)
    Stone Velocity Cone Angle=120
    Wood Gravity=&RandomRangeFloat(Minimum=-450, Maximum=-900)
    Wood Mesh Count=2
    Wood Scale=1.5
    Wood Velocity=&RandomRangeFloat(Minimum=80, Maximum=600)
    Wood Velocity Cone Angle=130
   )
   &AddVelocityInCone(
    Cone Angle=$Output.FN_PhysicalSurface.SurfaceResponseConeAngle
    Cone Axis=$System.DirectionVector
    Module.Use Velocity Falloff On Cone Axis=false
    Velocity Strength=$Output.FN_PhysicalSurface.SurfaceResponseVelocity
   )
   &InitialMeshOrientation(Use Rotation Vector=false)
   &AddRotationalVelocity(Rotation Rate=&RandomVector(Vector Scale=&RandomRangeFloat(Minimum=1, Maximum=2)))
   &ApplyInitialForces()
   /SetVariables(Particles.GravityOverrideValue=$Output.FN_PhysicalSurface.GravityOverride) {isSetParameters=true}
  ]
  particleUpdateScript [
   &ParticleState()
   &GravityForce(Gravity=$Particles.GravityOverrideValue)
   &Collision(
    Advanced Aging Rate=1
    Age Colliding Particles=false
    CPU Collision Trace Channel=ECC_WorldStatic
    Collision Enabled=&Boolean_AndOperation(
     Bool1=&SetBoolByFloatComparison(A=$Particles.Initial.Lifetime, B=1)
     Bool2=&SetBoolByFloatComparison(A=$Particles.Age, B=0.12)
    )
    Control Roll On Collision=true
    Friction=1
    Friction During a Bounce=0.25
    Max CPU Trace Length=100
    Module.Control Rest State Via Surface Normal=false
    Module.EnableMaxCollisionCount=false
    Module.Randomize Collision Normal Vector=true
    Randomize Collision Normal=0.25
    ScaleRateOfRotation=false
   ) {enabled=false}
   &SolveForcesAndVelocity(Clamp Velocity=false, Limit Acceleration=false)
   &ScaleMeshSize(Scale Factor=&VectorFromFloat(Value=&FloatFromCurve(FloatCurve=DataInterface() {propertyValues {Curve {keys [<at 0.6 1 RCIM_Cubic>, {arriveTangent=-6.308, interpMode=RCIM_Cubic, leaveTangent=-6.308, tangentMode=RCTM_User, time=1, value=0}]}}})))
   &DynamicMaterialParameters(Module.Param0WriteEnabled=true, Module.Param1WriteEnabled=true, Module.Param2WriteEnabled=false, Module.Param3WriteEnabled=false)
  ]
  rendererClasses [&NiagaraSpriteRendererProperties, &NiagaraMeshRendererProperties]
  renderers [
   {rendererClass=&NiagaraSpriteRendererProperties}
   {rendererClass=&NiagaraMeshRendererProperties, rendererIndex=1}
   {rendererClass=&NiagaraMeshRendererProperties, rendererIndex=2}
   {rendererClass=&NiagaraMeshRendererProperties, rendererIndex=3}
   {rendererClass=&NiagaraMeshRendererProperties, rendererIndex=4}
   {rendererClass=&NiagaraMeshRendererProperties, rendererIndex=5}
   {rendererClass=&NiagaraMeshRendererProperties, rendererIndex=6}
   {rendererClass=&NiagaraMeshRendererProperties, rendererIndex=7}
   {rendererClass=&NiagaraMeshRendererProperties, rendererIndex=8}
  ]
 }
 FN_Debug {
  bEnabled=false
  emitterUpdateScript [
   &EmitterState()
   &SpawnBurst_Instantaneous(Spawn Count=1, Use Spawn Probability=false)
  ]
  particleSpawnScript [/Niagara/Modules/Spawn/Initialization/InitializeParticle(
    Lifetime=0.5
    Module.Write Color=true
    Module.Write Lifetime=true
    Module.Write Mass=true
    Module.Write Material Random=true
    Module.Write Position=false
    Module.Write Scale=false
    Module.Write SpriteRotation=false
    Module.Write SpriteSize=true
    Position=&SimulationPosition()
    Sprite Size=<100 100>
   )]
  particleUpdateScript [
   &ParticleState()
   /SetVariables(Particles.Debug:NiagaraFloat=&MakeFloatFromInt(Int=$System.CurrentIndex)) {isSetParameters=true}
   &DynamicMaterialParameters(
    Index 0 Param 1=$Particles.Debug
    Module.Param0WriteEnabled=true
    Module.Param1WriteEnabled=true
    Module.Param2WriteEnabled=true
    Module.Param3WriteEnabled=true
   )
  ]
  rendererClasses [&NiagaraSpriteRendererProperties]
  renderers [{rendererClass=&NiagaraSpriteRendererProperties}]
 }
 FireEnergy {
  emitterSpawnScript [
   /SetVariables(Emitter.Spline_Base:NiagaraFloat=&FloatFromCurve(
     FloatCurve=DataInterface() {propertyValues {Curve {keys [
         {arriveTangent=3.225, interpMode=RCIM_Cubic, leaveTangent=3.225, tangentMode=RCTM_User, time=0, value=1}
         {arriveTangent=0.7836, interpMode=RCIM_Cubic, leaveTangent=0.7836, tangentMode=RCTM_User, time=0.2, value=1.6}
         <at 1 1.9 RCIM_Cubic>
        ]}}}
     CurveIndex=$Emitter.NormalizedLoopAge
     Scale Curve=&Multiply_Float(A=120, B=&FN_FrontEndFloatModifier(Gameplay=1, Front End=$System.FrontEndScale))
    )) {isSetParameters=true}
   /SetVariables(Emitter.Spline_Tip:NiagaraFloat=&FloatFromCurve(
     FloatCurve=DataInterface() {propertyValues {Curve {keys [
         {arriveTangent=1.0333, interpMode=RCIM_Cubic, leaveTangent=1.0333, tangentMode=RCTM_User, time=0, value=1}
         {arriveTangent=1.1134, interpMode=RCIM_Cubic, leaveTangent=1.1134, tangentMode=RCTM_User, time=0.3, value=1.3}
         <at 1 1.8 RCIM_Cubic>
        ]}}}
     CurveIndex=$Emitter.NormalizedLoopAge
     Scale Curve=&Multiply_Float(A=70, B=&FN_FrontEndFloatModifier(Gameplay=1, Front End=$System.FrontEndScale))
    )) {isSetParameters=true}
   /SetVariables(
    Emitter.Color1:LinearColor=&MergeHelper_ColorSwitch(
     0=<rgba 1 0.2 0 1>
     1=<rgba 0 0 1 1>
     2=?
     3=?
     Index=&MakeIntFromFloat(Float=$User.Variant)
     Num=::NewEnumerator0
    )
    Emitter.Color2:LinearColor=&MergeHelper_ColorSwitch(
     0=<rgba 2 0.6518 0.1 1>
     1=<rgba 0.1 1.05 2 1>
     2=?
     3=?
     Index=&MakeIntFromFloat(Float=$User.Variant)
     Num=::NewEnumerator0
    )
   ) {isSetParameters=true}
  ]
  emitterUpdateScript [
   &EmitterState(Life Cycle Mode=::NewEnumerator1, Loop Behavior=::NewEnumerator1, Loop Duration=1, UseLoopDelay=false)
   &SpawnBurst_Instantaneous(
    Spawn Probability=&MakeCustomFloatFromBool(
     Bool=&SetBoolByIntComparison(A=$System.CurrentIndex, B=3, Comparison Type=::NewEnumerator5)
     True Float=0
     False Float=1
    )
    Use Spawn Probability=true
    UseLoopCountLimit=false
   )
   &SpawnBurst_Instantaneous(
    Spawn Count=2
    Spawn Group=1
    Spawn Probability=&MakeCustomFloatFromBool(
     Bool=&SetBoolByIntComparison(A=$System.CurrentIndex, B=3, Comparison Type=::NewEnumerator5)
     True Float=0
     False Float=1
    )
    Use Spawn Probability=true
    UseLoopCountLimit=false
   ) {module=SpawnBurst_Instantaneous001}
  ]
  particleSpawnScript [
   /Niagara/Modules/Spawn/Initialization/V2/InitializeParticle(
    Lifetime=&MergeHelper_FloatSwitch(
     0=0.42
     1=&RandomRangeFloat(Minimum=0.3, Maximum=0.4)
     2=?
     3=?
     Index=$Emitter.SpawnGroup
     Num=::NewEnumerator0
    )
    Mesh Scale=&Multiply_VectorByFloat(Vector=<1.8 1 1>, Float=0.8)
    Mesh Scale Mode=::NewEnumerator4
    Sprite Size Max=<60 100>
    Sprite Size Min=<40 80>
    Sprite Size Mode=::NewEnumerator2
    UsePositionOffset=false
    Write Material Random=true
   )
   &AddVelocity(
    Cone Angle=140
    Module.UseConeFalloff=false
    Module.UseVelDistribution=false
    Override Seed=false
    Velocity Mode=::NewEnumerator2
    Velocity Speed=&Multiply_Float(
     A=&Multiply_Float(A=&RandomRangeFloat(Minimum=0.8, Maximum=1), B=100)
     B=&MergeHelper_FloatSwitch(0=0, 1=1, 2=?, 3=?, Index=$Emitter.SpawnGroup, Num=::NewEnumerator0)
    )
   )
   &AddVelocity(
    Velocity=$System.DirectionVector
    Velocity Speed Scale=&MergeHelper_FloatSwitch(
     0=&Multiply_Float(
      A=&RandomRangeFloat(Minimum=10, Maximum=20)
      B=&FN_FrontEndFloatModifier(Gameplay=8, Front End=$User.FrontEndScale)
     )
     1=&RandomRangeFloat(Minimum=200, Maximum=400)
     2=?
     3=?
     Index=$Emitter.SpawnGroup
     Num=::NewEnumerator0
    )
   ) {module=AddVelocity001}
   /SetVariables(Particles.VisibilityTag=&MergeHelper_IntSwitch(0=0, 1=1, 2=?, 3=?, 4=?, 5=?, Index=$Emitter.SpawnGroup, Num=::NewEnumerator0)) {isSetParameters=true}
   /SetVariables() {isSetParameters=true}
   &CameraOffset(Camera Offset Amount=&MergeHelper_FloatSwitch(0=0, 1=40, 2=?, 3=?, Index=$Emitter.SpawnGroup, Num=::NewEnumerator0))
  ]
  particleUpdateScript [
   &ParticleState()
   &SubUVAnimation(
    End Frame Range Override=30
    "Sprite Renderer "=DataInterface() {propertyValues {SpriteRenderer=&self:FireEnergy_0.NiagaraSpriteRendererProperties_1}}
    Start Frame Range Override=5
    UseEndFrame=true
    UseStartFrame=true
   )
   &ScaleMeshSize(Scale Factor=&Multiply_VectorByFloat(
     Vector=&MakeVector(
      X=&FloatFromCurve(FloatCurve=DataInterface() {propertyValues {Curve {keys [
           {arriveTangent=6.5629, interpMode=RCIM_Cubic, leaveTangent=6.5629, tangentMode=RCTM_User, time=0, value=0.5}
           {arriveTangent=0.5237, interpMode=RCIM_Cubic, leaveTangent=0.5237, tangentMode=RCTM_User, time=0.2, value=1.75}
           {arriveTangent=0.1793, interpMode=RCIM_Cubic, leaveTangent=0.1793, tangentMode=RCTM_User, time=1, value=2}
          ]}}})
      Y=1
      Z=1
     )
     Float=&FN_FrontEndFloatModifier(Gameplay=1, Front End=$System.FrontEndScale)
    ))
   &ScaleColor(
    Scale Alpha=&FloatFromCurve(
     FloatCurve=DataInterface() {propertyValues {Curve {keys [
         {arriveTangent=7.8189, interpMode=RCIM_Cubic, leaveTangent=7.8189, tangentMode=RCTM_User, time=0, value=0.2}
         {arriveTangent=0.0168, interpMode=RCIM_Cubic, leaveTangent=0.0168, tangentMode=RCTM_User, time=0.12, value=1}
         {arriveTangent=-1.6315, interpMode=RCIM_Cubic, leaveTangent=-1.6315, tangentMode=RCTM_User, time=0.4657, value=0.7806}
         <at 0.9 0 RCIM_Cubic>
        ]}}}
     Scale Curve=&Multiply_Float(A=2, B=&FNTimeOfDayScalar(Night Scale=0.5, Morning Scale=0.8, Evening Scale=0.8))
    )
    Scale RGB=&MakeVectorFromLinearColorRGB(Color=&MergeHelper_ColorSwitch(
      0=&ColorFromCurve(ColorCurve=DataInterface() {propertyValues {
         AlphaCurve {keys [<at 0 1 RCIM_Cubic>,<at 1 1 RCIM_Cubic>]}
         BlueCurve {keys [<at 0.2552 1 RCIM_Cubic>,<at 0.8097 0 RCIM_Cubic>]}
         GreenCurve {keys [<at 0.2552 1 RCIM_Cubic>,<at 0.8097 0.2 RCIM_Cubic>]}
         RedCurve {keys [<at 0.2552 1 RCIM_Cubic>,<at 0.8097 1 RCIM_Cubic>]}
        }})
      1=&ColorFromCurve(ColorCurve=DataInterface() {propertyValues {
         AlphaCurve {keys [<at 0 1 RCIM_Cubic>,<at 1 1 RCIM_Cubic>]}
         BlueCurve {keys [<at 0.2552 1 RCIM_Cubic>,<at 0.8097 1 RCIM_Cubic>]}
         GreenCurve {keys [<at 0.2552 1 RCIM_Cubic>,<at 0.8097 0.1667 RCIM_Cubic>]}
         RedCurve {keys [<at 0.2552 1 RCIM_Cubic>,<at 0.8097 0 RCIM_Cubic>]}
        }})
      2=?
      3=?
      Index=&MakeIntFromFloat(Float=$User.Variant)
      Num=::NewEnumerator0
     ))
    ScaleA=true
    ScaleRGB=true
   ) {module=ScaleColor002}
   &Drag(
    Drag=&MergeHelper_FloatSwitch(
     0=2
     1=&FloatFromCurve(FloatCurve=DataInterface() {propertyValues {Curve {keys [<at 0 0.075 RCIM_Cubic>,<at 1 1 RCIM_Cubic>]}}})
     2=?
     3=?
     Index=$Emitter.SpawnGroup
     Num=::NewEnumerator0
    )
    Use Linear Drag=true
    Use Rotational Drag=true
   )
   &SolveForcesAndVelocity(Clamp Velocity=false, Limit Acceleration=false)
   /SetVariables(Particles.Size003:Vector3f=&Multiply_VectorByFloat(
     Vector=&MakeVector(
      X=&FloatFromCurve(
       FloatCurve=DataInterface() {propertyValues {Curve {keys [
           {arriveTangent=6.5629, interpMode=RCIM_Cubic, leaveTangent=6.5629, tangentMode=RCTM_User, time=0, value=0.5}
           {arriveTangent=0.5237, interpMode=RCIM_Cubic, leaveTangent=0.5237, tangentMode=RCTM_User, time=0.2, value=1.75}
           {arriveTangent=0.1793, interpMode=RCIM_Cubic, leaveTangent=0.1793, tangentMode=RCTM_User, time=1, value=2}
          ]}}}
       Scale Curve=1.2
      )
      Y=1
      Z=1
     )
     Float=&Multiply_Float(A=0.8, B=&FN_FrontEndFloatModifier(Gameplay=1, Front End=$System.FrontEndScale))
    )) {isSetParameters=true}
   /SetVariables(Particles.GlowColor=&MakeLinearColorFromVectorAndFloat(
     "Vector (RGB)"=&MakeVectorFromLinearColorRGB(Color=&MergeHelper_ColorSwitch(
       0=&ColorFromCurve(ColorCurve=DataInterface() {propertyValues {
          AlphaCurve {keys []}
          BlueCurve {keys [<at 0 0 RCIM_Cubic>,<at 0.4105 0 RCIM_Cubic>]}
          GreenCurve {keys [<at 0 0.1424 RCIM_Cubic>,<at 0.4105 0.0104 RCIM_Cubic>]}
          RedCurve {keys [<at 0 0.4271 RCIM_Cubic>,<at 0.4105 0.0521 RCIM_Cubic>]}
         }})
       1=&ColorFromCurve(ColorCurve=DataInterface() {propertyValues {
          AlphaCurve {keys []}
          BlueCurve {keys [<at 0 0.4271 RCIM_Cubic>,<at 0.4105 0.0521 RCIM_Cubic>]}
          GreenCurve {keys [<at 0 0.2491 RCIM_Cubic>,<at 0.4105 0.0043 RCIM_Cubic>]}
          RedCurve {keys [<at 0 0 RCIM_Cubic>,<at 0.4105 0 RCIM_Cubic>]}
         }})
       2=?
       3=?
       Index=&MakeIntFromFloat(Float=$User.Variant)
       Num=::NewEnumerator0
      ))
     "Float (Alpha)"=&FloatFromCurve(
      FloatCurve=DataInterface() {propertyValues {Curve {keys [
          {arriveTangent=11.4287, interpMode=RCIM_Cubic, leaveTangent=11.4287, tangentMode=RCTM_User, time=0, value=0.45}
          {arriveTangent=0.0213, interpMode=RCIM_Cubic, leaveTangent=0.0213, tangentMode=RCTM_User, time=0.15, value=1}
          {arriveTangent=-1.9237, interpMode=RCIM_Cubic, leaveTangent=-1.9237, tangentMode=RCTM_User, time=0.4679, value=0.588}
          {arriveTangent=-1.1051, interpMode=RCIM_Cubic, leaveTangent=-1.1051, time=0.6984, value=0.2084}
          <at 1 0 RCIM_Cubic>
         ]}}}
      Scale Curve=0.45
     )
    )) {isSetParameters=true}
   &ScaleSpriteSize(
    Non-Uniform Curve Scale=&Multiply_Vector2D_ByFloat(Float=&FN_FrontEndFloatModifier(Gameplay=1, Front End=$System.FrontEndScale), Vector2D=<0.9 0.9>)
    Non-Uniform Curve Sprite Scale=DataInterface() {propertyValues {
      XCurve {keys [
        {arriveTangent=0.4232, interpMode=RCIM_Cubic, leaveTangent=0.4232, tangentMode=RCTM_User, time=0, value=0.8}
        {arriveTangent=-0.295, interpMode=RCIM_Cubic, leaveTangent=-0.295, time=0.6, value=0.975}
        {arriveTangent=0.037, interpMode=RCIM_Cubic, leaveTangent=0.037, tangentMode=RCTM_User, time=1.017, value=0.5}
       ]}
      YCurve {keys [
        <at 0 0.7 RCIM_Cubic>
        {arriveTangent=0.4424, interpMode=RCIM_Cubic, leaveTangent=0.4424, tangentMode=RCTM_User, time=0.3, value=0.8}
        {arriveTangent=0.1191, interpMode=RCIM_Cubic, leaveTangent=0.1191, tangentMode=RCTM_User, time=1, value=1.3}
       ]}
     }}
    Scale Sprite Size Mode=::NewEnumerator3
   )
   /SetVariables(Particles.FireColor:LinearColor=&MultiplyLinearColor(
     Linear Color=&MergeHelper_ColorSwitch(
      0=&ColorFromCurve(ColorCurve=DataInterface() {propertyValues {
         AlphaCurve {keys [<at 0 1 RCIM_Cubic>]}
         BlueCurve {keys [<at 0.3683 0 RCIM_Cubic>,<at 0.7291 0>,<at 0.798 0 RCIM_Cubic>]}
         GreenCurve {keys [<at 0.3683 2.132 RCIM_Cubic>,<at 0.7291 0.3663>,<at 0.798 0.0423 RCIM_Cubic>]}
         RedCurve {keys [<at 0.3683 10 RCIM_Cubic>,<at 0.7291 0.7379>,<at 0.798 0.0823 RCIM_Cubic>]}
        }})
      1=&ColorFromCurve(ColorCurve=DataInterface() {propertyValues {
         AlphaCurve {keys [<at 0 1 RCIM_Cubic>]}
         BlueCurve {keys [<at 0.3683 10 RCIM_Cubic>,<at 0.7291 0.7379>,<at 0.798 0.0823 RCIM_Cubic>]}
         GreenCurve {keys [<at 0.3683 6.6667 RCIM_Cubic>,<at 0.7291 0.123>,<at 0.798 0 RCIM_Cubic>]}
         RedCurve {keys [<at 0.3683 0 RCIM_Cubic>,<at 0.7291 0>,<at 0.798 0.0069 RCIM_Cubic>]}
        }})
      2=?
      3=?
      Index=&MakeIntFromFloat(Float=$User.Variant)
      Num=::NewEnumerator0
     )
     Scale Factor=&MakeVector4FromVectorAndFloat(
      "Vector (RGB)"=&MakeVectorFromLinearColorRGB(Color=&MultiplyLinearColor(
        Linear Color=&FNTimeOfDayColor(Night Color=<rgba 0.5 0.5 0.5 1>, Morning Color=<rgba 1 1 1 1>, Day Color=<rgba 1 1 1 1>, Evening Color=<rgba 0.8 0.8 0.8 1>)
        Scale Factor=&MakeVector4_FromLinearColor(LinearColor=<rgba 0.8698 0.8698 0.8698 1>)
       ))
      "Float (Alpha)"=&FloatFromCurve(FloatCurve=DataInterface() {propertyValues {Curve {keys [
           {arriveTangent=8.5399, interpMode=RCIM_Cubic, leaveTangent=8.5399, tangentMode=RCTM_User, time=0, value=0}
           {interpMode=RCIM_Cubic, tangentMode=RCTM_User, time=0.2, value=1}
           {arriveTangent=-1.8845, interpMode=RCIM_Cubic, leaveTangent=-1.8845, tangentMode=RCTM_User, time=0.7, value=0.3706}
           {arriveTangent=0.021, interpMode=RCIM_Cubic, leaveTangent=0.021, tangentMode=RCTM_User, time=1, value=0}
          ]}}})
     )
    )) {isSetParameters=true}
  ]
  rendererClasses [&NiagaraMeshRendererProperties, &NiagaraSpriteRendererProperties]
  renderers [
   {rendererClass=&NiagaraMeshRendererProperties, rendererIndex=0}
   {rendererClass=&NiagaraMeshRendererProperties, rendererIndex=1}
   {rendererClass=&NiagaraSpriteRendererProperties, rendererIndex=2}
  ]
 }
 FlashElements {
  emitterUpdateScript [
   &EmitterState()
   &SpawnBurst_Instantaneous(Spawn Group=1, Use Spawn Probability=false, UseLoopCountLimit=false) {module=SpawnBurst_Instantaneous001}
   &SpawnBurst_Instantaneous(
    Spawn Count=&RandomRangeInt(Minimum=6, Maximum=8)
    Use Spawn Probability=false
    UseLoopCountLimit=false
   )
  ]
  particleSpawnScript [
   &InitializeParticle(Color Mode=::NewEnumerator0, Lifetime=0.25, Sprite Rotation Mode=::NewEnumerator2, UsePositionOffset=false, Write Material Random=true)
   &ShapeLocation(MaskSpawnGroup=false, Override Seed=false, Sphere Radius=1, UseOwnerScale=true)
   &SpriteFacingAndAlignment(Sprite Alignment=$Particles.ShapeLocation.ShapeNormal, WriteAlignment=true, WriteFacing=false)
   &FN_PerParticleLOD(Enable SpriteSize Distance Scaling=true, Sprite Size Scale Far=<2.5 2.5>)
   /SetVariables(Particles.SpriteSize=<1 1>, Particles.VisibilityTag=&SpawnGroup()) {isSetParameters=true}
   /SetVariables(
    Particles.Flash Spike Color:LinearColor=&MultiplyLinearColor(
     Linear Color=&MergeHelper_ColorSwitch(
      0=&RandomRangeLinearColor(Minimum=<rgba 2 0.8985 0.0123 1>, Maximum=<rgba 2 0.8985 0.0123 1>)
      1=&MaskLinearColor(
       Mask Source=1
       Mask Parameter=$System.CurrentIndex
       Mask=0
       Masked Linear Color=&RandomRangeLinearColor(Minimum=<rgba 2 0.8985 0.0123 1>, Maximum=<rgba 2 0.8985 0.0123 1>)
       Passthrough Linear Color=&RandomRangeLinearColor(Minimum=<rgba 0.0123 1.3374 2 1>, Maximum=<rgba 0.1779 0.0123 2 1>)
      )
      2=?
      3=?
      Index=&MakeIntFromFloat(Float=$User.Variant)
      Num=::NewEnumerator0
     )
     Scale Factor=&MakeVector4_FromLinearColor(LinearColor=$System.TOD Color)
    )
    Particles.Flash Spike Lifetime:NiagaraFloat=&RandomRangeFloat(Minimum=0.05, Maximum=0.07)
    Particles.Flash Spike Sprite Size:Vector2f=&Multiply_Vector2D_ByFloat(
     Float=&FN_FrontEndFloatModifier(Gameplay=1, Front End=$System.FrontEndScale)
     Vector2D=&RandomRangeVector2D(Minimum=<3 55>, Maximum=<10 70>)
    )
   ) {isSetParameters=true}
   /SetVariables(
    Particles.Flash Star Cam Offset=15
    Particles.Flash Star Color:LinearColor=$System.TOD Color
    Particles.Flash Star Lifetime=0.15
    Particles.Flash Star Rotation:NiagaraFloat=&RandomRangeFloat(Minimum=0, Maximum=360)
    Particles.Flash Star Sprite Size:Vector2f=&Multiply_Vector2D_ByFloat(
     Float=&Multiply_Float(
      A=&MakeCustomFloatFromBool(Bool=$User.bIsFrontEnd, True Float=$System.FrontEndScale, False Float=1)
      B=&RandomRangeFloat(Minimum=120, Maximum=150)
     )
     Vector2D=$Particles.SpriteSize
    )
    Particles.Flash Star SubUV Index:NiagaraFloat=&RandomRangeFloat(Minimum=0, Maximum=3)
   ) {isSetParameters=true}
   /SetVariables(
    Particles.Flash Flare Color:LinearColor=$System.TOD Color
    Particles.Flash Flare Lifetime=0.15
    Particles.Flash Flare Sprite Size:Vector2f=&Multiply_Vector2D_ByFloat(
     Float=&Multiply_Float(A=0.8, B=&FN_FrontEndFloatModifier(Gameplay=1, Front End=$System.FrontEndScale))
     Vector2D=<180 180>
    )
   ) {isSetParameters=true}
   &DynamicMaterialParameters(
    Index 0 Param 1=0.2
    Index 0 Param 2=1
    Module.Index0Param0Write=true
    Module.Index0Param1Write=true
    Module.Index0Param2Write=true
    Module.Index0Param3Write=true
   )
   &CameraOffset(Camera Offset Amount=15)
  ]
  particleUpdateScript [
   &ParticleState()
   /SetVariables(
    Particles.Flash Flare Age:NiagaraFloat=&Divide_Float(A=$Particles.Age, B=$Particles.Initial.Flash Flare Lifetime)
    Particles.Flash Spike Age:NiagaraFloat=&Divide_Float(A=$Particles.Age, B=$Particles.Initial.Flash Spike Lifetime)
    Particles.Flash Star Age:NiagaraFloat=&Divide_Float(A=$Particles.Age, B=$Particles.Initial.Flash Star Lifetime)
   ) {isSetParameters=true}
   /SetVariables(
    Particles.Flash Flare Color:LinearColor=&MakeLinearColorFromVectorAndFloat(
     "Vector (RGB)"=&Multiply_Vector(
      A=&Multiply_Vector(
       A=&MakeVectorFromLinearColorRGB(Color=$Particles.Initial.Flash Flare Color)
       B=&MergeHelper_VectorSwitch(
        0=&MakeVectorFromLinearColorRGB(Color=&MergeHelper_ColorSwitch(
          0=&ColorFromCurve(
           ColorCurve=DataInterface() {propertyValues {
             AlphaCurve {keys [<at 0 1 RCIM_Cubic>,<at 1 0 RCIM_Cubic>]}
             BlueCurve {keys [<at 0.0198 0>,<at 0.491 0>,<at 1 0>]}
             GreenCurve {keys [<at 0.0198 1.2233>,<at 0.491 0.2>,<at 1 0.3>]}
             RedCurve {keys [<at 0.0198 2>,<at 0.491 2>,<at 1 3>]}
            }}
           CurveIndex=$Particles.Flash Flare Age
          )
          1=&ColorFromCurve(
           ColorCurve=DataInterface() {propertyValues {
             AlphaCurve {keys [<at 0 1 RCIM_Cubic>,<at 1 0 RCIM_Cubic>]}
             BlueCurve {keys [<at 0.0198 0>,<at 0.491 0>,<at 1 0>]}
             GreenCurve {keys [<at 0.0198 1.2233>,<at 0.491 0.2>,<at 1 0.0333>]}
             RedCurve {keys [<at 0.0198 2>,<at 0.491 2>,<at 1 1>]}
            }}
           CurveIndex=$Particles.Flash Flare Age
          )
          2=&ColorFromCurve(
           ColorCurve=DataInterface() {propertyValues {
             AlphaCurve {keys [<at 0 1 RCIM_Cubic>,<at 1 0 RCIM_Cubic>]}
             BlueCurve {keys [<at 0.0198 0>,<at 0.491 0>,<at 1 0>]}
             GreenCurve {keys [<at 0.0198 1.2233>,<at 0.491 0.3333>,<at 1 0.75>]}
             RedCurve {keys [<at 0.0198 2>,<at 0.491 2>,<at 1 3>]}
            }}
           CurveIndex=$Particles.Flash Flare Age
          )
          3=&ColorFromCurve(
           ColorCurve=DataInterface() {propertyValues {
             AlphaCurve {keys [<at 0 1 RCIM_Cubic>,<at 1 0 RCIM_Cubic>]}
             BlueCurve {keys [<at 0.0198 0>,<at 0.491 0>,<at 1 0>]}
             GreenCurve {keys [<at 0.0198 1.2233>,<at 0.491 0.2>,<at 1 0.3>]}
             RedCurve {keys [<at 0.0198 2>,<at 0.491 2>,<at 1 3>]}
            }}
           CurveIndex=$Particles.Flash Flare Age
          )
          Index=$System.CurrentIndex
          Num=::NewEnumerator2
         ))
        1=&MakeVectorFromLinearColorRGB(Color=&MergeHelper_ColorSwitch(
          0=&ColorFromCurve(
           ColorCurve=DataInterface() {propertyValues {
             AlphaCurve {keys [<at 0 1 RCIM_Cubic>,<at 1 0 RCIM_Cubic>]}
             BlueCurve {keys [<at 0.0198 0>,<at 0.491 0>,<at 1 0>]}
             GreenCurve {keys [<at 0.0198 1.2233>,<at 0.491 0.2>,<at 1 0.3>]}
             RedCurve {keys [<at 0.0198 2>,<at 0.491 2>,<at 1 3>]}
            }}
           CurveIndex=$Particles.Flash Flare Age
          )
          1=&ColorFromCurve(
           ColorCurve=DataInterface() {propertyValues {
             AlphaCurve {keys [<at 0 1 RCIM_Cubic>,<at 1 0 RCIM_Cubic>]}
             BlueCurve {keys [<at 0.0198 2>,<at 0.491 2>,<at 1 1>]}
             GreenCurve {keys [<at 0.0198 1.3333>,<at 0.491 0.3333>,<at 1 0>]}
             RedCurve {keys [<at 0.0198 0>,<at 0.491 0>,<at 1 0.0833>]}
            }}
           CurveIndex=$Particles.Flash Flare Age
          )
          2=&ColorFromCurve(
           ColorCurve=DataInterface() {propertyValues {
             AlphaCurve {keys [<at 0 1 RCIM_Cubic>,<at 1 0 RCIM_Cubic>]}
             BlueCurve {keys [<at 0.0198 2>,<at 0.491 2>,<at 1 1>]}
             GreenCurve {keys [<at 0.0198 1.3333>,<at 0.491 0>,<at 1 0>]}
             RedCurve {keys [<at 0.0198 0>,<at 0.491 0.1667>,<at 1 0.0833>]}
            }}
           CurveIndex=$Particles.Flash Flare Age
          )
          3=&ColorFromCurve(
           ColorCurve=DataInterface() {propertyValues {
             AlphaCurve {keys [<at 0 1 RCIM_Cubic>,<at 1 0 RCIM_Cubic>]}
             BlueCurve {keys [<at 0.0198 2>,<at 0.491 2>,<at 1 1>]}
             GreenCurve {keys [<at 0.0198 1.3333>,<at 0.491 0.3333>,<at 1 0>]}
             RedCurve {keys [<at 0.0198 0>,<at 0.491 0>,<at 1 0.0833>]}
            }}
           CurveIndex=$Particles.Flash Flare Age
          )
          Index=$System.CurrentIndex
          Num=::NewEnumerator2
         ))
        2=?
        3=?
        Index=&MakeIntFromFloat(Float=$User.Variant)
        Num=::NewEnumerator0
       )
      )
      B=&VectorFromFloat(Value=&FN_FrontEndFloatModifier(Gameplay=1.5, Front End=1))
     )
     "Float (Alpha)"=&Multiply_Float(
      A=1
      B=&FloatFromCurve(
       FloatCurve=DataInterface() {propertyValues {Curve {keys [{arriveTangent=-1.9727, interpMode=RCIM_Cubic, time=0.1, value=1}, <at 1 0 RCIM_Cubic>]}}}
       CurveIndex=$Particles.Flash Flare Age
      )
     )
    )
    Particles.Flash Spike Color:LinearColor=&MakeLinearColorFromVectorAndFloat(
     "Vector (RGB)"=&MakeVectorFromLinearColorRGB(Color=$Particles.Initial.Flash Spike Color)
     "Float (Alpha)"=&FloatFromCurve(FloatCurve=DataInterface() {propertyValues {Curve {keys [<at 0.7 1>,<at 1 0>]}}}, CurveIndex=$Particles.Flash Spike Age)
    )
    Particles.Flash Spike Sprite Size:Vector2f=&Multiply_Vector2D_ByFloat(
     Float=&FloatFromCurve(FloatCurve=DataInterface() {propertyValues {Curve {keys [<at 0 0.6>,<at 1 2>]}}}, CurveIndex=$Particles.Flash Spike Age, Scale Curve=0.8)
     Vector2D=$Particles.Initial.Flash Spike Sprite Size
    )
    Particles.Flash Star Color:LinearColor=&MultiplyLinearColor(
     Linear Color=$Particles.Initial.Flash Star Color
     Scale Factor=&MakeVector4_FromLinearColor(LinearColor=&MergeHelper_ColorSwitch(
       0=&MergeHelper_ColorSwitch(
        0=&ColorFromCurve(
         ColorCurve=DataInterface() {propertyValues {
           AlphaCurve {keys [<at 0.2935 1>,<at 0.9862 0>]}
           BlueCurve {keys [<at 0.0344 0>,<at 0.4768 0>]}
           GreenCurve {keys [<at 0.0344 1.5>,<at 0.4768 0.3307>]}
           RedCurve {keys [<at 0.0344 3>,<at 0.4768 3>]}
          }}
         CurveIndex=$Particles.Flash Star Age
        )
        1=&ColorFromCurve(
         ColorCurve=DataInterface() {propertyValues {
           AlphaCurve {keys [<at 0.2935 1>,<at 0.9862 0>]}
           BlueCurve {keys [<at 0.0344 0>,<at 0.4768 0>]}
           GreenCurve {keys [<at 0.0344 1.5>,<at 0.4768 0.05>]}
           RedCurve {keys [<at 0.0344 3>,<at 0.4768 3>]}
          }}
         CurveIndex=$Particles.Flash Star Age
        )
        2=&ColorFromCurve(
         ColorCurve=DataInterface() {propertyValues {
           AlphaCurve {keys [<at 0.2935 1>,<at 0.9862 0>]}
           BlueCurve {keys [<at 0.0344 0>,<at 0.4768 0>]}
           GreenCurve {keys [<at 0.0344 1.5>,<at 0.4768 0.3307>]}
           RedCurve {keys [<at 0.0344 3>,<at 0.4768 3>]}
          }}
         CurveIndex=$Particles.Flash Star Age
        )
        3=&ColorFromCurve(
         ColorCurve=DataInterface() {propertyValues {
           AlphaCurve {keys [<at 0.2935 1>,<at 0.9862 0>]}
           BlueCurve {keys [<at 0.0344 0>,<at 0.4768 0>]}
           GreenCurve {keys [<at 0.0344 1.5>,<at 0.4768 0.3307>]}
           RedCurve {keys [<at 0.0344 3>,<at 0.4768 3>]}
          }}
         CurveIndex=$Particles.Flash Star Age
        )
        Index=$System.CurrentIndex
        Num=::NewEnumerator2
       )
       1=&MergeHelper_ColorSwitch(
        0=&ColorFromCurve(
         ColorCurve=DataInterface() {propertyValues {
           AlphaCurve {keys [<at 0.2935 1>,<at 0.9862 0>]}
           BlueCurve {keys [<at 0.0344 0>,<at 0.4768 0>]}
           GreenCurve {keys [<at 0.0344 1.5>,<at 0.4768 0.3307>]}
           RedCurve {keys [<at 0.0344 3>,<at 0.4768 3>]}
          }}
         CurveIndex=$Particles.Flash Star Age
        )
        1=&ColorFromCurve(
         ColorCurve=DataInterface() {propertyValues {
           AlphaCurve {keys [<at 0.2935 1>,<at 0.9862 0>]}
           BlueCurve {keys [<at 0.0344 3>,<at 0.4768 3>]}
           GreenCurve {keys [<at 0.0344 2>,<at 0.4768 0.5>]}
           RedCurve {keys [<at 0.0344 0>,<at 0.4768 0>]}
          }}
         CurveIndex=$Particles.Flash Star Age
        )
        2=&ColorFromCurve(
         ColorCurve=DataInterface() {propertyValues {
           AlphaCurve {keys [<at 0.2935 1>,<at 0.9862 0>]}
           BlueCurve {keys [<at 0.0344 3>,<at 0.4768 3>]}
           GreenCurve {keys [<at 0.0344 2>,<at 0.4768 0>]}
           RedCurve {keys [<at 0.0344 0>,<at 0.4768 0>]}
          }}
         CurveIndex=$Particles.Flash Star Age
        )
        3=&ColorFromCurve(
         ColorCurve=DataInterface() {propertyValues {
           AlphaCurve {keys [<at 0.2935 1>,<at 0.9862 0>]}
           BlueCurve {keys [<at 0.0344 3>,<at 0.4768 3>]}
           GreenCurve {keys [<at 0.0344 2>,<at 0.4768 0.5>]}
           RedCurve {keys [<at 0.0344 0>,<at 0.4768 0>]}
          }}
         CurveIndex=$Particles.Flash Star Age
        )
        Index=$System.CurrentIndex
        Num=::NewEnumerator2
       )
       2=?
       3=?
       Index=&MakeIntFromFloat(Float=$User.Variant)
       Num=::NewEnumerator0
      ))
    )
    Particles.Flash Star Sprite Size:Vector2f=&Multiply_Vector2D_ByFloat(
     Float=&FloatFromCurve(
      FloatCurve=DataInterface() {propertyValues {Curve {keys [
          {arriveTangent=2.7013, interpMode=RCIM_Cubic, leaveTangent=2.7013, tangentMode=RCTM_User, time=0, value=0}
          {interpMode=RCIM_Cubic, tangentMode=RCTM_User, time=1, value=1}
         ]}}}
      CurveIndex=$Particles.Flash Star Age
      Scale Curve=0.8
     )
     Vector2D=$Particles.Initial.Flash Star Sprite Size
    )
   ) {isSetParameters=true}
   &FN_Light_Attributes(
    Color=&MultiplyLinearColor(
     Linear Color=&MergeHelper_ColorSwitch(
      0=&MergeHelper_ColorSwitch(0=<rgba 2 0.5361 0 1>, 1=<rgba 2 0.0333 0 1>, 2=<rgba 2 1 0 1>, 3=<rgba 2 0.5361 0 1>, Index=$System.CurrentIndex, Num=::NewEnumerator2)
      1=&MergeHelper_ColorSwitch(0=<rgba 2 0.5361 0 1>, 1=<rgba 0 0.3333 2 1>, 2=<rgba 0.1667 0 2 1>, 3=<rgba 0 0.3333 2 1>, Index=$System.CurrentIndex, Num=::NewEnumerator2)
      2=?
      3=?
      Index=&MakeIntFromFloat(Float=$User.Variant)
      Num=::NewEnumerator0
     )
     Scale Factor=&MakeVector4_FromLinearColor(LinearColor=&LerpLinearColors(StartColor=<rgba 1 1 1 1>, EndColor=$System.TOD Color, LerpFactor=0.5))
    )
    Create Color Attribute=true
    Create Exponent Attribute=true
    Create Position Attribute=true
    Create Radius Attribute=true
    Exponent=2
    Radius=&FN_FrontEndFloatModifier(Gameplay=90, Front End=90)
    Scale Color=&FloatFromCurve(
     FloatCurve=DataInterface() {propertyValues {Curve {keys [
         {arriveTangent=-0.5504, interpMode=RCIM_Cubic, leaveTangent=-0.5504, tangentMode=RCTM_User, time=0, value=1}
         {arriveTangent=-0.5426, interpMode=RCIM_Cubic, leaveTangent=-0.5426, tangentMode=RCTM_User, time=0.55, value=0}
        ]}}}
     Scale Curve=7
    )
   )
  ]
  rendererClasses [&NiagaraSpriteRendererProperties, &NiagaraLightRendererProperties]
  renderers [
   {rendererClass=&NiagaraSpriteRendererProperties}
   {rendererClass=&NiagaraSpriteRendererProperties, rendererIndex=1}
   {rendererClass=&NiagaraSpriteRendererProperties, rendererIndex=2}
   {rendererClass=&NiagaraLightRendererProperties, rendererIndex=3}
  ]
 }
 Flash_Flare {
  emitterUpdateScript [
   &EmitterState()
   &SpawnBurst_Instantaneous(Use Spawn Probability=true, UseLoopCountLimit=false)
  ]
  particleSpawnScript [
   &InitializeParticle(
    Color=&MergeHelper_ColorSwitch(
     0=&MergeHelper_ColorSwitch(0=<rgba 2 0.6405 0 1>, 1=<rgba 2 0.2 0 1>, 2=<rgba 2 0.6405 0 1>, 3=<rgba 2 0.7765 0.2 1>, Index=$System.CurrentIndex, Num=::NewEnumerator2)
     1=&MergeHelper_ColorSwitch(0=<rgba 2 0.6405 0 1>, 1=<rgba 0 0.8333 2 1>, 2=<rgba 0 0.8333 2 1>, 3=<rgba 0 0.8333 2 1>, Index=$System.CurrentIndex, Num=::NewEnumerator2)
     2=?
     3=?
     Index=&MakeIntFromFloat(Float=$User.Variant)
     Num=::NewEnumerator0
    )
    Lifetime=0.18
    Sprite Size=&FN_FrontEndVector2Modifier(Front End:Vector2f=<40 150>, Gameplay:Vector2f=<80 240>)
    Sprite Size Mode=::NewEnumerator4
    UsePositionOffset=false
    Write Material Random=true
   )
   &CameraOffset(Camera Offset Amount=30)
   &DynamicMaterialParameters(
    Index 0 Param 1=0.2
    Index 0 Param 2=1
    Index 3 Param 1=1.5
    Index 3 Param 2=1
    Module.Param0WriteEnabled=true
    Module.Param0WriteEnabled3=true
    Module.Param1WriteEnabled=true
    Module.Param1WriteEnabled3=true
    Module.Param2WriteEnabled=true
    Module.Param2WriteEnabled3=true
    Module.Param3WriteEnabled=true
    Module.Param3WriteEnabled3=true
    Write Parameter Index 3=true
   )
   &AddVelocity(
    Rotation Coordinate Space=::Simulation
    Velocity=$System.DirectionVector
    Velocity Speed Scale=&RandomRangeFloat(Minimum=200, Maximum=300)
   )
  ]
  particleUpdateScript [
   &ParticleState()
   &ScaleSpriteSize(
    Non-Uniform Curve Scale=<0.8 0.8>
    Non-Uniform Curve Sprite Scale=DataInterface() {propertyValues {
      XCurve {keys [{tangentMode=RCTM_User, time=0, value=0}, <at 0.2 1.075>, {tangentMode=RCTM_User, time=1, value=-0.1}]}
      YCurve {keys [{tangentMode=RCTM_User, time=0, value=0}, <at 0.1 1.025>, {tangentMode=RCTM_User, time=1, value=0.8}]}
     }}
    Scale Sprite Size Mode=::NewEnumerator3
   )
   &SolveForcesAndVelocity(Clamp Velocity=false, Limit Acceleration=false)
  ]
  rendererClasses [&NiagaraSpriteRendererProperties]
  renderers [{rendererClass=&NiagaraSpriteRendererProperties}]
 }
 ImpactSmoke {
  emitterUpdateScript [
   &EmitterState()
   &SpawnBurst_Instantaneous(
    Spawn Count=&RandomRangeInt(Minimum=2, Maximum=3)
    Use Spawn Probability=false
    UseLoopCountLimit=false
   )
  ]
  particleSpawnScript [
   &InitializeParticle(
    Color Maximum=<rgba 0.76 1 1 0.2>
    Color Minimum=<rgba 0.8 0.917 1 0.15>
    Color Mode=::NewEnumerator2
    Lifetime Max=0.4
    Lifetime Min=0.3
    Lifetime Mode=::NewEnumerator1
    Sprite Size Mode=::NewEnumerator3
    Uniform Sprite Size=&Multiply_Float(
     A=&MakeCustomFloatFromBool(Bool=$User.bIsFrontEnd, True Float=$System.FrontEndScale, False Float=1)
     B=&RandomRangeFloat(Minimum=100, Maximum=130)
    )
    UsePositionOffset=false
    Write Material Random=true
   )
   &CameraOffset(Camera Offset Amount=&RandomRangeFloat(Minimum=40, Maximum=50, Evaluation Type=::NewEnumerator1))
   &ShapeLocation(
    Hemisphere Distribution=<1 0.5>
    MaskSpawnGroup=false
    Override Seed=false
    Rotation Angle=90
    Rotation Mode=::NewEnumerator1
    Sphere Radius=&MakeCustomFloatFromBool(Bool=$User.bIsFrontEnd, True Float=5, False Float=10)
    UseOwnerScale=true
   )
   &AddVelocity(Velocity=&Multiply_VectorByFloat(Vector=$Particles.ShapeLocation.ShapeVector, Float=10))
  ]
  particleUpdateScript [
   &ParticleState()
   &ScaleSpriteSize(Uniform Curve Sprite Scale=DataInterface() {propertyValues {Curve {keys [
        {arriveTangent=1.7761, interpMode=RCIM_Cubic, leaveTangent=1.7761, tangentMode=RCTM_User, time=0, value=0}
        {interpMode=RCIM_Cubic, tangentMode=RCTM_User, time=1, value=1}
       ]}}})
   &ScaleColor(
    Scale Alpha=&FloatFromCurve(FloatCurve=DataInterface() {propertyValues {Curve {keys [
         {arriveTangent=2.4358, interpMode=RCIM_Cubic, leaveTangent=2.4358, tangentMode=RCTM_User, time=0, value=0}
         <at 0.25 0.5 RCIM_Cubic>
         <at 1 0 RCIM_Cubic>
        ]}}})
    Scale RGB=&VectorFromCurve(VectorCurve=DataInterface() {propertyValues {XCurve {keys [<at 0 1>,<at 1 1>]}, YCurve {keys [<at 0 1>,<at 1 1>]}, ZCurve {keys [<at 0 1>,<at 1 1>]}}})
    ScaleA=true
    ScaleRGB=true
   )
   &DynamicMaterialParameters(
    Index 0 Param 1=&Add_Float(
     A=&RandomRangeFloat(Minimum=0, Maximum=1)
     B=&FloatFromCurve(
      FloatCurve=DataInterface() {propertyValues {Curve {keys [
          <at 0 0 RCIM_Cubic>
          {arriveTangent=1.5947, interpMode=RCIM_Cubic, leaveTangent=1.5947, tangentMode=RCTM_User, time=0.2557, value=0.5257}
          <at 1 1 RCIM_Cubic>
         ]}}}
      CurveIndex=$Particles.Age
      Scale Curve=0.25
     )
    )
    Module.Param0WriteEnabled=true
    Module.Param1WriteEnabled=true
    Module.Param2WriteEnabled=true
    Module.Param3WriteEnabled=true
   ) {module=DynamicMaterialParameters001}
   &SolveForcesAndVelocity(Clamp Velocity=false, Limit Acceleration=false)
  ]
  rendererClasses [&NiagaraSpriteRendererProperties]
  renderers [
   {rendererClass=&NiagaraSpriteRendererProperties}
   {rendererClass=&NiagaraSpriteRendererProperties, rendererIndex=1}
  ]
 }
 Sparks {
  emitterUpdateScript [
   &EmitterState()
   &SpawnBurst_Instantaneous(
    Spawn Count=&MakeCustomIntFromBool(Bool=$User.bIsFrontEndPreview, True INT=30, False INT=&RandomRangeInt(Minimum=15, Maximum=20))
    Use Spawn Probability=false
    UseLoopCountLimit=false
   )
  ]
  particleSpawnScript [
   &InitializeParticle(
    Color=&MergeHelper_ColorSwitch(
     0=&MergeHelper_ColorSwitch(
      0=&LerpLinearColors(
       StartColor=<rgba 2 0.9414 0.2536 1>
       EndColor=<rgba 2 0.6548 0.3045 1>
       LerpFactor=&RandomRangeFloat(Minimum=0, Maximum=1)
      )
      1=&LerpLinearColors(
       StartColor=<rgba 2 0.9414 0.2536 1>
       EndColor=<rgba 2 0.6548 0.3045 1>
       LerpFactor=&RandomRangeFloat(Minimum=0, Maximum=1)
      )
      2=<rgba 2 0.6548 0.3045 1>
      3=<rgba 2 0.8985 0.0123 1>
      Index=$System.CurrentIndex
      Num=::NewEnumerator2
     )
     1=&MergeHelper_ColorSwitch(
      0=&LerpLinearColors(
       StartColor=<rgba 2 0.9414 0.2536 1>
       EndColor=<rgba 2 0.6548 0.3045 1>
       LerpFactor=&RandomRangeFloat(Minimum=0, Maximum=1)
      )
      1=&LerpLinearColors(
       StartColor=<rgba 0.2536 1.4179 2 1>
       EndColor=<rgba 0.3045 0.8696 2 1>
       LerpFactor=&RandomRangeFloat(Minimum=0, Maximum=1)
      )
      2=<rgba 0.2536 1.4179 2 1>
      3=<rgba 0.2536 1.4179 2 1>
      Index=$System.CurrentIndex
      Num=::NewEnumerator2
     )
     2=?
     3=?
     Index=&MakeIntFromFloat(Float=$User.Variant)
     Num=::NewEnumerator0
    )
    Lifetime Max=0.45
    Lifetime Min=0.05
    Lifetime Mode=::NewEnumerator1
    Sprite Size Mode=::NewEnumerator3
    Uniform Sprite Size=&Multiply_Float(
     A=&FN_FrontEndFloatModifier(Gameplay=1, Front End=$System.FrontEndScale)
     B=&RandomRangeFloat(Minimum=3, Maximum=5)
    )
    UsePositionOffset=false
    Write Material Random=true
   )
   &ShapeLocation(
    MaskSpawnGroup=false
    Override Seed=false
    Sphere Radius=&Multiply_Float(A=0.8, B=&FN_FrontEndFloatModifier(Gameplay=0, Front End=$System.FrontEndScale))
    UseOwnerScale=true
   )
   &CameraOffset(Camera Offset Amount=25)
   &AddVelocity(
    Cone Angle=180
    Module.UseConeFalloff=false
    Module.UseVelDistribution=false
    Override Seed=false
    Velocity Mode=::NewEnumerator2
    Velocity Speed=&Multiply_Float(
     A=&RandomRangeFloat(Minimum=400, Maximum=750)
     B=&Multiply_Float(A=&FN_FrontEndFloatModifier(Gameplay=1, Front End=$System.FrontEndScale), B=0.8)
    )
   )
  ]
  particleUpdateScript [
   &ParticleState()
   &AccelerationForce(Acceleration=&RandomRangeVector(Minimum=<0 0 -500>, Maximum=<0 0 -500>))
   &Drag(Drag=&RandomRangeFloat(Minimum=5, Maximum=8), Use Linear Drag=true, Use Rotational Drag=true)
   &SolveForcesAndVelocity(Clamp Velocity=false, Limit Acceleration=false)
   &ScaleSpriteSizeBySpeed(Max Scale Factor=<0.5 10>, Min Scale Factor=<0.333 1>)
   &ScaleSpriteSize(
    Uniform Curve Scale=&Multiply_Float(A=1, B=&FN_FrontEndFloatModifier(Gameplay=1, Front End=0.5))
    Uniform Curve Sprite Scale=DataInterface() {propertyValues {Curve {keys [{arriveTangent=-6.725, interpMode=RCIM_Cubic, leaveTangent=-6.725, tangentMode=RCTM_User, time=0.65, value=1}, <at 1 0 RCIM_Cubic>]}}}
   )
   &ScaleColor(
    Scale RGB=&Multiply_Vector(
     A=&VectorFromFloat(Value=&FloatFromCurve(
       FloatCurve=DataInterface() {propertyValues {Curve {keys [
           {interpMode=RCIM_Cubic, tangentMode=RCTM_User, time=0, value=1}
           <at 0.3 0.2>
           {interpMode=RCIM_Cubic, tangentMode=RCTM_User, time=1, value=0}
          ]}}}
       Scale Curve=10
      ))
     B=&MakeVectorFromLinearColorRGB(Color=&MergeHelper_ColorSwitch(
       0=<rgba 1 1 1 1>
       1=<rgba 1 1 1 1>
       2=<rgba 1 1 1 1>
       3=&ColorFromCurve(ColorCurve=DataInterface() {propertyValues {
          AlphaCurve {keys [<at 0 1 RCIM_Cubic>,<at 1 0 RCIM_Cubic>]}
          BlueCurve {keys [<at 0.4507 1 RCIM_Cubic>,<at 0.7557 0.05 RCIM_Cubic>]}
          GreenCurve {keys [<at 0.4507 1 RCIM_Cubic>,<at 0.7557 0.1608 RCIM_Cubic>]}
          RedCurve {keys [<at 0.4507 1 RCIM_Cubic>,<at 0.7557 1 RCIM_Cubic>]}
         }})
       Index=$System.CurrentIndex
       Num=::NewEnumerator2
      ))
    )
    ScaleA=true
    ScaleRGB=true
   )
   /SetVariables(
    Particles.GlowColor=&MakeLinearColorFromVectorAndFloat(
     "Vector (RGB)"=&MakeVectorFromLinearColorRGB(Color=&MergeHelper_ColorSwitch(
       0=&MergeHelper_ColorSwitch(0=<rgba 1 0.2015 0 1>, 1=<rgba 1 0.05 0 1>, 2=<rgba 1 0.2015 0 1>, 3=<rgba 1 0.19 0.1 1>, Index=$System.CurrentIndex, Num=::NewEnumerator2)
       1=&MergeHelper_ColorSwitch(0=<rgba 1 0.2015 0 1>, 1=<rgba 0 0.1667 1 1>, 2=<rgba 0 0.1667 1 1>, 3=<rgba 0 0.1667 1 1>, Index=$System.CurrentIndex, Num=::NewEnumerator2)
       2=?
       3=?
       Index=&MakeIntFromFloat(Float=$User.Variant)
       Num=::NewEnumerator0
      ))
     "Float (Alpha)"=&FloatFromCurve(FloatCurve=DataInterface() {propertyValues {Curve {keys [<at 0.5 1>,<at 1 0>]}}}, Scale Curve=0.08)
    )
    Particles.GlowSize=&Multiply_Vector2D(A=$Particles.SpriteSize, B=<8 2>)
   ) {isSetParameters=true}
  ]
  rendererClasses [&NiagaraSpriteRendererProperties]
  renderers [
   {rendererClass=&NiagaraSpriteRendererProperties}
   {rendererClass=&NiagaraSpriteRendererProperties, rendererIndex=1}
  ]
 }
}
summary {
 systemName=NS_Pickaxe_DerangedMile_Impact
 emitters [
  {emitterName=FlashElements, rendererClasses [&NiagaraSpriteRendererProperties, &NiagaraLightRendererProperties]}
  {emitterName=ImpactSmoke, rendererClasses [&NiagaraSpriteRendererProperties]}
  {emitterName=DirectionalStreaks, rendererClasses [&NiagaraSpriteRendererProperties]}
  {
   emitterName=Directional_FN__PhysMatDebris
   rendererClasses [&NiagaraSpriteRendererProperties, &NiagaraMeshRendererProperties]
  }
  {emitterName=DirectionalSparks, rendererClasses [&NiagaraSpriteRendererProperties]}
  {emitterName=Sparks, rendererClasses [&NiagaraSpriteRendererProperties]}
  {emitterName=Flash_Flare, rendererClasses [&NiagaraSpriteRendererProperties]}
  {emitterName=FireEnergy, rendererClasses [&NiagaraMeshRendererProperties, &NiagaraSpriteRendererProperties]}
  {emitterName=Claw, rendererClasses [&NiagaraSpriteRendererProperties]}
  {bEnabled=false, emitterName=FN_Debug, rendererClasses [&NiagaraSpriteRendererProperties]}
 ]
 userVariables [
  User.bIsFrontEnd {type=NiagaraBool}
  User.bIsFrontEndPreview {type=NiagaraBool}
  User.CurrentIndex {type=NiagaraInt32}
  User.FrontEndScale {type=NiagaraFloat, defaultValue=0.6, description=""}
  User.PhysicalSurface {type=NiagaraInt32}
  User.ReactivityIsOn {type=NiagaraFloat, defaultValue=1, description=""}
  User.SwingSectionIndex {type=NiagaraInt32}
  User.Variant {type=NiagaraFloat, defaultValue=0, description=""}
 ]
}
