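# Niagara swing trail effect for the DerangedMile pickaxe, built from five anim-trail ribbon emitters that follow weapon bones or sockets.
# System.CurrentIndex switches trails: Base, BaseTip and Dark (also gated by swing section) serve index 0, while Color and Tip count and palette vary by index.
# User.Variant picks warm or blue colors for Color and Tip.
# Swing fade and time-of-day values scale the trails, and Color uses two renderers with glow and dark color attributes.
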NiagaraSystem /BRCosmetics/Weapons/FORT_Melee/Pickaxe_DerangedMile/FX/NS_Pickaxe_DerangedMile_Swing
dependencies {
 usedDataInterfaces [&NiagaraDataInterfaceCurve, &NiagaraDataInterfaceSocketReader, &NiagaraDataInterfaceColorCurve]
 usedDynamicInputs [
  &ColorFromCurve
  &FNTimeOfDayScalar
  &Add_Float
  &Sine
  &SetBoolByFloatComparison
 ]
 usedModules [
  &self:SystemSpawnScript.SystemScriptSource.SystemScriptGraph.NiagaraNodeAssignment_0.SetVariables_#1
  &SystemState
  &self:SystemSpawnScript.SystemScriptSource.SystemScriptGraph.NiagaraNodeAssignment_4.SetVariables_#2
  &self:SystemSpawnScript.SystemScriptSource.SystemScriptGraph.NiagaraNodeAssignment_5.SetVariables_#3
  &self:SystemSpawnScript.SystemScriptSource.SystemScriptGraph.NiagaraNodeAssignment_1.SetVariables_#4
  &self:SystemSpawnScript.SystemScriptSource.SystemScriptGraph.NiagaraNodeAssignment_2.SetVariables_#5
 ]
}
emitters {
 Base {
  emitterSpawnScript [&FN_RibbonIDOffsetUtility()]
  emitterUpdateScript [
   &EmitterState()
   &FN_SpawnAnimTrail(Number of Anim Trails:NiagaraInt32=&MergeHelper_IntSwitch(0=1, 1=0, 2=0, 3=0, 4=0, 5=?, Index=$System.CurrentIndex, Num=::NewEnumerator3))
  ]
  particleSpawnScript [
   &FN_InitializeAnimTrailBoneIndecies()
   &FN_InitializeAnimTrail(Color=<rgba 1 1 1 0.5>, Lifetime=0.2)
  ]
  particleUpdateScript [
   &SolveForcesAndVelocity(Clamp Velocity=false, Limit Acceleration=false)
   &FN_UpdateAnimTrail() {module=FN_UpdateAnimTrail001}
   &ScaleColor(
    Scale Alpha=&FloatFromCurve(
     FloatCurve=DataInterface() {propertyValues {Curve {keys [{interpMode=RCIM_Cubic, tangentMode=RCTM_User, time=0, value=1}, {interpMode=RCIM_Cubic, tangentMode=RCTM_User, time=0.4, value=0}]}}}
     Scale Curve=$System.SwingGeneralFade
    )
    ScaleA=true
    ScaleRGB=true
   )
   &ScaleColor(
    Scale Alpha=$System.TODFloat
    Scale RGB=&MakeVectorFromLinearColorRGB(Color=$System.TODColor)
    ScaleA=true
    ScaleRGB=true
   ) {module=ScaleColor002}
  ]
  rendererClasses [&NiagaraRibbonRendererProperties]
  renderers [{rendererClass=&NiagaraRibbonRendererProperties}]
 }
 BaseTip {
  emitterUpdateScript [
   &EmitterState()
   &FN_SpawnAnimTrail(Number of Anim Trails:NiagaraInt32=&MergeHelper_IntSwitch(0=1, 1=0, 2=0, 3=0, 4=0, 5=?, Index=$System.CurrentIndex, Num=::NewEnumerator3))
  ]
  particleSpawnScript [
   &FN_InitializeAnimTrailBoneIndecies(First Bone or Socket Index=2)
   &FN_InitializeAnimTrail(Lifetime=0.2)
   /SetVariables(Particles.RibbonWidthDark:NiagaraFloat=&Multiply_Float(A=$Particles.RibbonWidth, B=0.9)) {isSetParameters=true, module=SetVariables_#6}
   &AddVelocity(Velocity=<0 1000 0>) {enabled=false}
  ]
  particleUpdateScript [
   &SolveForcesAndVelocity(Clamp Velocity=false, Limit Acceleration=false) {enabled=false}
   &FN_UpdateAnimTrail(Locked Particle RibbonWidth=1) {module=FN_UpdateAnimTrail001}
   &ScaleColor(
    Scale Alpha=&FloatFromCurve(
     FloatCurve=DataInterface() {propertyValues {Curve {keys [{arriveTangent=-0.0018, interpMode=RCIM_Cubic, leaveTangent=-0.2406, tangentMode=RCTM_Break, time=0.4, value=1}, <at 1 0 RCIM_Cubic>]}}}
     Scale Curve=&Multiply_Float(A=&Multiply_Float(A=$System.SwingGeneralFade, B=$System.TODFloat), B=1)
    )
    Scale RGB=&MakeVectorFromLinearColorRGB(Color=$System.TODColor)
    ScaleA=true
    ScaleRGB=true
   ) {enabled=false}
   /SetVariables(Particles.ColorTip:LinearColor=&MakeLinearColorFromVectorAndFloat(
     "Vector (RGB)"=&MakeVectorFromLinearColorRGB(Color=&MultiplyLinearColor(
       Linear Color=<rgba 0.056 0.0699 0.1 1>
       Scale Factor=&MakeVector4_FromLinearColor(LinearColor=$System.TODColor)
      ))
     "Float (Alpha)"=&FloatFromCurve(
      FloatCurve=DataInterface() {propertyValues {Curve {keys [{arriveTangent=-0.0018, interpMode=RCIM_Cubic, leaveTangent=-0.2406, tangentMode=RCTM_Break, time=0.4, value=1}, <at 1 0 RCIM_Cubic>]}}}
      Scale Curve=&Multiply_Float(A=&Multiply_Float(A=$System.SwingGeneralFade, B=$System.TODFloat), B=1)
     )
    )) {isSetParameters=true, module=SetVariables_#7}
  ]
  rendererClasses [&NiagaraRibbonRendererProperties]
  renderers [{rendererClass=&NiagaraRibbonRendererProperties}]
 }
 Color {
  emitterUpdateScript [
   &EmitterState()
   &FN_SpawnAnimTrail(Number of Anim Trails:NiagaraInt32=&MakeCustomIntFromBool(
     Bool=&SetBoolByIntComparison(A=$System.CurrentIndex, B=0, Comparison Type=::NewEnumerator4)
     True INT=0
     False INT=1
    ))
  ]
  particleSpawnScript [
   &FN_InitializeAnimTrailBoneIndecies()
   &FN_InitializeAnimTrail(AnimTrailWidth=0.95, Lifetime=0.2)
   /SetVariables(Particles.RibbonWidthDark:NiagaraFloat=&Multiply_Float(A=$Particles.RibbonWidth, B=0.9)) {isSetParameters=true, module=SetVariables_#6}
  ]
  particleUpdateScript [
   &FN_UpdateAnimTrail(Locked Particle RibbonWidth=1) {module=FN_UpdateAnimTrail001}
   &ScaleColor(
    Scale Alpha=&FloatFromCurve(
     FloatCurve=DataInterface() {propertyValues {Curve {keys [{arriveTangent=-0.0018, interpMode=RCIM_Cubic, leaveTangent=-0.2406, tangentMode=RCTM_Break, time=0.4, value=1}, <at 1 0 RCIM_Cubic>]}}}
     Scale Curve=&Multiply_Float(A=&Multiply_Float(A=$System.SwingGeneralFade, B=$System.TODFloat), B=1)
    )
    Scale RGB=&MakeVectorFromLinearColorRGB(Color=$System.TODColor)
    ScaleA=true
    ScaleRGB=true
   ) {enabled=false}
   /SetVariables(
    Particles.ColorGlow:LinearColor=&MakeLinearColorFromVectorAndFloat(
     "Vector (RGB)"=&MakeVectorFromLinearColorRGB(Color=&MultiplyLinearColor(
       Linear Color=&MergeHelper_ColorSwitch(
        0=&MergeHelper_ColorSwitch(
         0=<rgba 1 0 0 1>
         1=&LerpLinearColors(
          StartColor=<rgba 1.3 0.4023 0 1>
          EndColor=<rgba 1 0 0 1>
          LerpFactor=&FloatFromCurve(
           FloatCurve=DataInterface() {propertyValues {Curve {keys [{interpMode=RCIM_Cubic, tangentMode=RCTM_Break, time=0.3, value=0}, {interpMode=RCIM_Cubic, tangentMode=RCTM_Break, time=1, value=1}]}}}
           CurveIndex=&ReturnNormalizedExecIndex()
          )
         )
         2=&LerpLinearColors(
          StartColor=<rgba 1.2 0.6854 0.0822 1>
          EndColor=<rgba 1 0.2232 0.0232 1>
          LerpFactor=&FloatFromCurve(
           FloatCurve=DataInterface() {propertyValues {Curve {keys [{interpMode=RCIM_Cubic, tangentMode=RCTM_Break, time=0, value=0}, {interpMode=RCIM_Cubic, tangentMode=RCTM_Break, time=0.5, value=1}]}}}
           CurveIndex=&ReturnNormalizedExecIndex()
          )
         )
         3=&LerpLinearColors(
          StartColor=<rgba 1 0.7913 0.402 1>
          EndColor=<rgba 1 0.1144 0.0075 1>
          LerpFactor=&FloatFromCurve(
           FloatCurve=DataInterface() {propertyValues {Curve {keys [{interpMode=RCIM_Cubic, tangentMode=RCTM_Break, time=0, value=0}, {interpMode=RCIM_Cubic, tangentMode=RCTM_Break, time=0.5, value=1}]}}}
           CurveIndex=&ReturnNormalizedExecIndex()
          )
         )
         Index=$System.CurrentIndex
         Num=::NewEnumerator2
        )
        1=&MergeHelper_ColorSwitch(
         0=<rgba 1 0 0 1>
         1=&LerpLinearColors(
          StartColor=<rgba 0.0497 0.4443 0.8388 1>
          EndColor=<rgba 0.0953 0.0561 0.7991 1>
          LerpFactor=&FloatFromCurve(
           FloatCurve=DataInterface() {propertyValues {Curve {keys [{interpMode=RCIM_Cubic, tangentMode=RCTM_Break, time=0.3, value=0}, {interpMode=RCIM_Cubic, tangentMode=RCTM_Break, time=1, value=1}]}}}
           CurveIndex=&ReturnNormalizedExecIndex()
          )
         )
         2=&LerpLinearColors(
          StartColor=<rgba 0.0497 0.5758 0.8388 1>
          EndColor=<rgba 0.0953 0.0561 0.7991 1>
          LerpFactor=&FloatFromCurve(
           FloatCurve=DataInterface() {propertyValues {Curve {keys [{interpMode=RCIM_Cubic, tangentMode=RCTM_Break, time=0, value=0}, {interpMode=RCIM_Cubic, tangentMode=RCTM_Break, time=0.5, value=1}]}}}
           CurveIndex=&ReturnNormalizedExecIndex()
          )
         )
         3=&LerpLinearColors(
          StartColor=<rgba 0.195 0.9317 1.3 1>
          EndColor=<rgba 0.0953 0.0561 0.7991 1>
          LerpFactor=&FloatFromCurve(
           FloatCurve=DataInterface() {propertyValues {Curve {keys [{interpMode=RCIM_Cubic, tangentMode=RCTM_Break, time=0, value=0}, {interpMode=RCIM_Cubic, tangentMode=RCTM_Break, time=0.5, value=1}]}}}
           CurveIndex=&ReturnNormalizedExecIndex()
          )
         )
         Index=$System.CurrentIndex
         Num=::NewEnumerator2
        )
        2=?
        3=?
        Index=&MakeIntFromFloat(Float=$User.Variant)
        Num=::NewEnumerator0
       )
       Scale Factor=&MakeVector4_FromLinearColor(LinearColor=$System.TODColor)
      ))
     "Float (Alpha)"=&FloatFromCurve(
      FloatCurve=DataInterface() {propertyValues {Curve {keys [{arriveTangent=-0.0018, interpMode=RCIM_Cubic, leaveTangent=-1.5833, tangentMode=RCTM_Break, time=0, value=1}, <at 0.4 0 RCIM_Cubic>]}}}
      Scale Curve=&Multiply_Float(
       A=&Multiply_Float(A=$System.SwingGeneralFade, B=$System.TODFloat)
       B=&MergeHelper_FloatSwitch(0=1, 1=0.7, 2=1, 3=1, Index=$System.CurrentIndex, Num=::NewEnumerator2)
      )
     )
    )
    Particles.DarkColor:LinearColor=&MakeLinearColorFromVectorAndFloat(
     "Vector (RGB)"=&MakeVectorFromLinearColorRGB(Color=<rgba 0.02 0.02 0.02 2>)
     "Float (Alpha)"=&FloatFromCurve(
      FloatCurve=DataInterface() {propertyValues {Curve {keys [{arriveTangent=-0.0018, interpMode=RCIM_Cubic, leaveTangent=-0.2406, tangentMode=RCTM_Break, time=0.4, value=1}, <at 1 0 RCIM_Cubic>]}}}
      Scale Curve=&Multiply_Float(
       A=&Multiply_Float(A=$System.SwingGeneralFade, B=$System.TODFloat)
       B=&MergeHelper_FloatSwitch(0=0, 1=0.5, 2=1, 3=1, 4=1, Index=$System.CurrentIndex, Num=::NewEnumerator3)
      )
     )
    )
   ) {isSetParameters=true, module=SetVariables_#7}
  ]
  rendererClasses [&NiagaraRibbonRendererProperties]
  renderers [
   {rendererClass=&NiagaraRibbonRendererProperties}
   {rendererClass=&NiagaraRibbonRendererProperties, rendererIndex=1}
  ]
 }
 Dark {
  emitterUpdateScript [
   &EmitterState()
   &FN_SpawnAnimTrail(Number of Anim Trails:NiagaraInt32=&MergeHelper_IntSwitch(
     0=&MergeHelper_IntSwitch(0=1, 1=0, 2=1, 3=1, 4=?, 5=?, Index=$User.SwingSectionIndex, Num=::NewEnumerator2)
     1=0
     2=0
     3=0
     4=0
     5=?
     Index=$System.CurrentIndex
     Num=::NewEnumerator3
    ))
  ]
  particleSpawnScript [
   &FN_InitializeAnimTrailBoneIndecies(First Bone or Socket Index=3, Second Bone or Socket Index=4)
   &FN_InitializeAnimTrail(Color=<rgba 5 2 0.5 1>, Lifetime=0.2)
   /SetVariables(Particles.RibbonWidthDark:NiagaraFloat=&Multiply_Float(A=$Particles.RibbonWidth, B=0.9)) {isSetParameters=true, module=SetVariables_#6}
  ]
  particleUpdateScript [
   &FN_UpdateAnimTrail(Locked Particle RibbonWidth=1) {module=FN_UpdateAnimTrail001}
   /SetVariables(Particles.ColorTip:LinearColor=&MakeLinearColorFromVectorAndFloat(
     "Vector (RGB)"=&MakeVectorFromLinearColorRGB(Color=&MultiplyLinearColor(
       Linear Color=<rgba 0.024 0.0292 0.03 1>
       Scale Factor=&MakeVector4_FromLinearColor(LinearColor=$System.TODColor)
      ))
     "Float (Alpha)"=&FloatFromCurve(
      FloatCurve=DataInterface() {propertyValues {Curve {keys [{arriveTangent=-0.0018, interpMode=RCIM_Cubic, leaveTangent=-0.2406, tangentMode=RCTM_Break, time=0.4, value=1}, <at 1 0 RCIM_Cubic>]}}}
      Scale Curve=&Multiply_Float(A=&Multiply_Float(A=$System.SwingGeneralFade, B=$System.TODFloat), B=0.5)
     )
    )) {isSetParameters=true, module=SetVariables_#7}
  ]
  rendererClasses [&NiagaraRibbonRendererProperties]
  renderers [{rendererClass=&NiagaraRibbonRendererProperties}]
 }
 Tip {
  emitterUpdateScript [
   &EmitterState()
   &FN_SpawnAnimTrail(Number of Anim Trails:NiagaraInt32=&MakeCustomIntFromBool(
     Bool=&SetBoolByIntComparison(A=$System.CurrentIndex, B=0, Comparison Type=::NewEnumerator4)
     True INT=0
     False INT=1
    ))
  ]
  particleSpawnScript [
   &FN_InitializeAnimTrailBoneIndecies(First Bone or Socket Index=2)
   &FN_InitializeAnimTrail(Color=<rgba 5 2 0.5 1>, Lifetime=0.2)
   /SetVariables(Particles.RibbonWidthDark:NiagaraFloat=&Multiply_Float(A=$Particles.RibbonWidth, B=0.9)) {isSetParameters=true, module=SetVariables_#6}
   &AddVelocity(Velocity=<0 1000 0>) {enabled=false}
  ]
  particleUpdateScript [
   &SolveForcesAndVelocity(Clamp Velocity=false, Limit Acceleration=false) {enabled=false}
   &FN_UpdateAnimTrail(Locked Particle RibbonWidth=1) {module=FN_UpdateAnimTrail001}
   &ScaleColor(
    Scale Alpha=&FloatFromCurve(
     FloatCurve=DataInterface() {propertyValues {Curve {keys [{arriveTangent=-0.0018, interpMode=RCIM_Cubic, leaveTangent=-0.2406, tangentMode=RCTM_Break, time=0.4, value=1}, <at 1 0 RCIM_Cubic>]}}}
     Scale Curve=&Multiply_Float(A=&Multiply_Float(A=$System.SwingGeneralFade, B=$System.TODFloat), B=1)
    )
    Scale RGB=&MakeVectorFromLinearColorRGB(Color=$System.TODColor)
    ScaleA=true
    ScaleRGB=true
   ) {enabled=false}
   /SetVariables(Particles.ColorTip:LinearColor=&MakeLinearColorFromVectorAndFloat(
     "Vector (RGB)"=&MakeVectorFromLinearColorRGB(Color=&MergeHelper_ColorSwitch(
       0=&MultiplyLinearColor(
        Linear Color=&MergeHelper_ColorSwitch(0=<rgba 1 0 0 1>, 1=<rgba 1 0 0 1>, 2=<rgba 1 0.2232 0.0232 1>, 3=<rgba 1 0.1352 0.0021 1>, Index=$System.CurrentIndex, Num=::NewEnumerator2)
        Scale Factor=&MakeVector4_FromLinearColor(LinearColor=$System.TODColor)
       )
       1=&MultiplyLinearColor(
        Linear Color=&MergeHelper_ColorSwitch(
         0=<rgba 0.0953 0.0561 0.7991 1>
         1=<rgba 0.0953 0.0561 0.7991 1>
         2=<rgba 0.0953 0.0561 0.7991 1>
         3=<rgba 0.0953 0.0561 0.7991 1>
         Index=$System.CurrentIndex
         Num=::NewEnumerator2
        )
        Scale Factor=&MakeVector4_FromLinearColor(LinearColor=$System.TODColor)
       )
       2=?
       3=?
       Index=&MakeIntFromFloat(Float=$User.Variant)
       Num=::NewEnumerator0
      ))
     "Float (Alpha)"=&FloatFromCurve(
      FloatCurve=DataInterface() {propertyValues {Curve {keys [{arriveTangent=-0.0018, interpMode=RCIM_Cubic, leaveTangent=-0.2406, tangentMode=RCTM_Break, time=0.4, value=1}, <at 1 0 RCIM_Cubic>]}}}
      Scale Curve=&Multiply_Float(A=&Multiply_Float(A=$System.SwingGeneralFade, B=$System.TODFloat), B=1)
     )
    )) {isSetParameters=true, module=SetVariables_#7}
  ]
  rendererClasses [&NiagaraRibbonRendererProperties]
  renderers [{rendererClass=&NiagaraRibbonRendererProperties}]
 }
}
summary {
 systemName=NS_Pickaxe_DerangedMile_Swing
 emitters [
  {emitterName=Color, rendererClasses [&NiagaraRibbonRendererProperties]}
  {emitterName=Tip, rendererClasses [&NiagaraRibbonRendererProperties]}
  {emitterName=Base, rendererClasses [&NiagaraRibbonRendererProperties]}
  {emitterName=BaseTip, rendererClasses [&NiagaraRibbonRendererProperties]}
  {emitterName=Dark, rendererClasses [&NiagaraRibbonRendererProperties]}
 ]
 userVariables [
  User.bIsFrontEnd {type=NiagaraBool}
  User.bIsFrontEndPreview {type=NiagaraBool}
  User.CurrentIndex {type=NiagaraInt32}
  User.ReactivityIsOn {type=NiagaraFloat, defaultValue=1, description=""}
  User.RibbonIDOffset {type=NiagaraInt32}
  User.SwingSectionIndex {type=NiagaraInt32}
  User.Variant {type=NiagaraFloat, defaultValue=0, description=""}
 ]
}
